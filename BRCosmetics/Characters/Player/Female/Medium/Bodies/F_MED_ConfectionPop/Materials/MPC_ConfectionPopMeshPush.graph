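# Material function for the ConfectionPop female body cosmetic that pushes masked sleeve regions of the mesh through World Position Offset.
# It builds a local-space offset from a push direction and strength, scaled separately by animated left and right arm blend parameters.
# FX texture channels supply per-sleeve masks, and the offset is added to incoming material attributes.
# A static switch gates the effect, and Low and Medium quality pass attributes through unchanged.
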MaterialFunction /BRCosmetics/Characters/Player/Female/Medium/Bodies/F_MED_ConfectionPop/Materials/MPC_ConfectionPopMeshPush
graph {nodes [
  n1 = FunctionInput() {inputType=FunctionInput_MaterialAttributes}
  n2 = GetMaterialAttributes(None=@n1) {attributeGetTypes [#1, #2]}
  n3 = SetMaterialAttributes!(
   MaterialAttributes=@n2.MaterialAttributes
   WorldPositionOffset=Add(A=@n2.WorldPositionOffset, B=NamedRerouteUsage())
   EmissiveColor=@n2.MaterialAttributes
  ) {attributeSetTypes [#1, #2]}
  FunctionOutput(None=StaticSwitchParameter(True=QualitySwitch(Default=@n3, Low=@n1, High=@n3, Medium=@n1, Epic=@n3), False=@n1) {group=1 - ConfectionPopMeshPush, parameterName=UseConfectionPopMeshPush, sortPriority=0})
  n4 = PreSkinnedPosition()
  NamedRerouteDeclaration(None=ShaderStageSwitch(PixelShader=VertexInterpolator(VS=@n4).PS, VertexShader=@n4)) {name=PSLP, nodeColor=<rgba 1 0 0.4922 1>}
  n5 = NamedRerouteUsage()
  n6 = NamedRerouteUsage()
  n7 = VectorParameter() {group=1 - ConfectionPopMeshPush, parameterName=PushDirection, defaultValue=<rgba 1 0 0 1>}
  n8 = Multiply(
   A=Transform(None=Multiply(A=@n6, B=@n7.RGB)) {transformSourceType=TRANSFORMSOURCE_Local}
   B=ScalarParameter() {defaultValue=0.08, group=1 - ConfectionPopMeshPush, parameterName=Push_Strenght, sliderMax=10}
  )
  n9 = Multiply(
   A=Saturate(None=ScalarParameter() {desc=Animated Parameters, group=1 - ConfectionPopMeshPush, parameterName=Arm_Blend_L, sliderMax=1, sortPriority=40})
   B=@n8
  )
  n10 = Multiply(
   A=@n8
   B=Saturate(None=ScalarParameter() {desc=Animated Parameters, group=1 - ConfectionPopMeshPush, parameterName=Arm_Blend_R, sliderMax=1, sortPriority=40})
  )
  n11 = NamedRerouteUsage()
  NamedRerouteDeclaration(None=Add(A=Multiply(A=@n5, B=@n9), B=Multiply(A=@n10, B=@n11))) {name=WPO, nodeColor=<rgba 0.0078 0 1 1>}
  n12 = TextureSample() {
   samplerType=SAMPLERTYPE_Masks
   texture=/BRCosmetics/Characters/Player/Female/Medium/Bodies/F_MED_ConfectionPop/Textures/T_ConfectionPop_Body_FX
  }
  n13 = FunctionInput(Preview=@n12.RGBA) {inputName=FX, inputType=FunctionInput_Vector4, sortPriority=1}
  n14 = ChannelMaskParameter(None=@n13) {group=1 - ConfectionPopMeshPush, maskChannel=Green, parameterName=ArmMask_Left, defaultValue=<rgba 0 1 0 0>}
  NamedRerouteDeclaration(None=@n14.R) {name=SleevesMask_L, nodeColor=<rgba 1 0.2876 0.2378 1>}
  n15 = ChannelMaskParameter(None=@n13) {group=1 - ConfectionPopMeshPush, maskChannel=Blue, parameterName=ArmMask_Right, defaultValue=<rgba 0 0 1 0>}
  NamedRerouteDeclaration(None=@n15.R) {name=SleevesMask_R, nodeColor=<rgba 0.3833 0.4599 1 1>}
 ]}
