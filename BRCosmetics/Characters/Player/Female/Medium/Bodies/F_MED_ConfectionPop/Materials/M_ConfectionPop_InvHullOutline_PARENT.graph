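# Parent material for the inverted-hull anime outline on the ConfectionPop female player body, a masked, two-sided, unlit skinned-mesh shader.
# The shared anime outline function feeds a ConfectionPop mesh-push function, then cloth-sim WPO and tech-art override functions.
# Character post-FX finishes the chain, fed by a constant, pre-skinned position interpolated for the pixel stage, distance blend and Fresnel inputs.
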Material /BRCosmetics/Characters/Player/Female/Medium/Bodies/F_MED_ConfectionPop/Materials/M_ConfectionPop_InvHullOutline_PARENT
graph {
 properties {
  bAutomaticallySetUsageInEditor=false
  bEnableResponsiveAA=true
  bUseMaterialAttributes=true
  bUsedWithClothing=true
  bUsedWithMeshDeformer=true
  bUsedWithMorphTargets=true
  bUsedWithSkeletalMesh=true
  blendMode=BLEND_Masked
  pixelDepthOffsetMode=PDOM_Legacy
  shadingModel=MSM_Unlit
  twoSided=true
  usedShadingModels=Unlit
 }
 nodes [
  Composite() {subgraphName=Construct Base Inputs}
  n1 = MaterialFunctionCall() {materialFunction=/Game/Global/Materials/Character_Parent_2023/Anime_Parent_2023/MaterialFunctions/MFC_AnimeOutline_InvHull_2023}
  n2 = MaterialFunctionCall!(In=@n1.Result, FX=@n1.Result) {materialFunction=&/MPC_ConfectionPopMeshPush}
  n3 = Constant()
  MaterialExpressionReroute_2 = Reroute(None=@n3)
  n4 = PreSkinnedPosition()
  n5 = ShaderStageSwitch(PixelShader=VertexInterpolator(VS=@n4).PS, VertexShader=@n4)
  MaterialExpressionReroute_8 = Reroute(None=@n5)
  n6 = MaterialFunctionCall!(Material) {materialFunction=/Game/Packages/Fortress_SharedMaterials/Base_Material_Functions/CharacterBrightnessandScale_Balance}
  MaterialExpressionReroute_9 = Reroute(None=@n6.Distance Blend)
  n7 = Fresnel() {baseReflectFraction=0, exponent=1}
  MaterialExpressionReroute_10 = Reroute(None=@n7)
  MaterialFunctionCall(
   Material=MaterialFunctionCall(In=MaterialFunctionCall!(Material=@n2.Result, SkinFX_Mask) {materialFunction=/Game/Global/MaterialFunctions/MF_ClothSimWPO_v2}.Result) {materialFunction=&MF_TechArtOverrides}.Result
   M=@MaterialExpressionReroute_2.M
   Pre Skinned Local Position=@MaterialExpressionReroute_8.Pre Skinned Local Position
   Distance Blend=@MaterialExpressionReroute_9.Distance Blend
   Fresnel=@MaterialExpressionReroute_10.Fresnel
  ) {materialFunction=&MF_CharacterParent_2023_PostFX}
  PinBase()
  PinBase!(M=@n3, Pre Skinned Local Position=@n5, Distance Blend=@n6.Distance Blend, Fresnel=@n7) {reroutePins [
    M {expression=&self:MaterialExpressionReroute_2}
    Pre Skinned Local Position {expression=&self:MaterialExpressionReroute_8}
    Distance Blend {expression=&self:MaterialExpressionReroute_9}
    Fresnel {expression=&self:MaterialExpressionReroute_10}
   ]}
 ]
}
