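# Physics asset for the backpack mesh of the F_MED_ConfectionPop female player body, setting up physics for its dyn_blouse bones.
# Seven four-segment capsule chains (center, plus front, side and back pairs) attach through base bodies to spine_02, mostly with angular-limited joints.
# Capsules on the pelvis, spine and thighs supply body volumes.
# Angularly locked cross-constraints tie front, side and back end segments together on each side, linking neighboring chains.
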PhysicsAsset /BRCosmetics/Characters/Player/Female/Medium/Bodies/F_MED_ConfectionPop/Meshes/F_MED_ConfectionPop_Backpack_Physics
bodies {
 dyn_blouse_01_c [{shapeType=Capsule, length=10, radius=5, shapeName=dyn_blouse_01_c_capsule, center=<0 0 5>, rotation=<rot 0 0 -90>}]
 dyn_blouse_02_c [{shapeType=Capsule, length=10, radius=5, shapeName=dyn_blouse_02_c_capsule, center=<2.8283 0 3>, rotation=<rot 0 0 90>}]
 dyn_blouse_03_c [{shapeType=Capsule, length=10, radius=6, shapeName=dyn_blouse_03_c_capsule, center=<2.6389 0 6>, rotation=<rot 0 0 90>}]
 dyn_blouse_04_c [{shapeType=Capsule, length=10, radius=6, shapeName=dyn_blouse_04_c_capsule, center=<5 0 3>, rotation=<rot 0 0 86.2027>}]
 dyn_blouse_base_01_c [{shapeType=Capsule, length=0.01, radius=1, shapeName=dyn_blouse_base_01_c_capsule}]
 dyn_blouse_base_bk_l [{shapeType=Capsule, length=0.505, radius=0.505, shapeName=dyn_blouse_base_bk_l_capsule, rotation=<rot 0 -90 0>}]
 dyn_blouse_base_bk_r [{shapeType=Capsule, length=0.505, radius=0.505, shapeName=dyn_blouse_base_bk_r_capsule, rotation=<rot 0 -90 0>}]
 dyn_blouse_base_fr_l [{shapeType=Capsule, length=0.505, radius=0.505, shapeName=dyn_blouse_base_fr_l_capsule, rotation=<rot 0 -90 0>}]
 dyn_blouse_base_fr_r [{shapeType=Capsule, length=0.505, radius=0.505, shapeName=dyn_blouse_base_fr_r_capsule, rotation=<rot 0 -90 0>}]
 dyn_blouse_base_side_l [{shapeType=Capsule, length=0.505, radius=0.505, shapeName=dyn_blouse_base_side_l_capsule, rotation=<rot 0 -90 0>}]
 dyn_blouse_base_side_r [{shapeType=Capsule, length=0.505, radius=0.505, shapeName=dyn_blouse_base_side_r_capsule, rotation=<rot 0 -90 0>}]
 dyn_blouse_bk_01_l [{shapeType=Capsule, length=10, radius=5, shapeName=dyn_blouse_bk_01_l_capsule, center=<0 0 5>, rotation=<rot 0 0 -90>}]
 dyn_blouse_bk_01_r [{shapeType=Capsule, length=10, radius=5, shapeName=dyn_blouse_bk_01_r_capsule, center=<0 0 5>, rotation=<rot 0 0 -90>}]
 dyn_blouse_bk_02_l [{shapeType=Capsule, length=10, radius=5, shapeName=dyn_blouse_bk_02_l_capsule, center=<2.8283 0 3>, rotation=<rot 0 0 90>}]
 dyn_blouse_bk_02_r [{shapeType=Capsule, length=10, radius=5, shapeName=dyn_blouse_bk_02_r_capsule, center=<2.8283 0 3>, rotation=<rot 0 0 90>}]
 dyn_blouse_bk_03_l [{shapeType=Capsule, length=10, radius=6, shapeName=dyn_blouse_bk_03_l_capsule, center=<2.6389 0 4.5>, rotation=<rot 0 0 90>}]
 dyn_blouse_bk_03_r [{shapeType=Capsule, length=10, radius=6, shapeName=dyn_blouse_bk_03_r_capsule, center=<2.6389 0 5>, rotation=<rot 0 0 90>}]
 dyn_blouse_bk_04_l [{
   shapeType=Capsule
   length=20.9413
   radius=8
   shapeName=dyn_blouse_bk_04_l_capsule
   center=<5 -0.093 3>
   rotation=<rot -0.0004 0.0002 91.9197>
  }]
 dyn_blouse_bk_04_r [{shapeType=Capsule, length=10, radius=6, shapeName=dyn_blouse_bk_04_r_capsule, center=<5 0 3>, rotation=<rot 0 0 99.5861>}]
 dyn_blouse_fr_01_l [{shapeType=Capsule, length=10, radius=5, shapeName=dyn_blouse_fr_01_l_capsule, center=<0 -0.0134 5.7>, rotation=<rot 0.0001 0 -92.6187>}]
 dyn_blouse_fr_01_r [{shapeType=Capsule, length=10, radius=5, shapeName=dyn_blouse_fr_01_r_capsule, center=<0 0 5>, rotation=<rot 0 0 -90>}]
 dyn_blouse_fr_02_l [{shapeType=Capsule, length=10, radius=5, shapeName=dyn_blouse_fr_02_l_capsule, center=<2.8283 0 5.08>, rotation=<rot 0 0 90>}]
 dyn_blouse_fr_02_r [{shapeType=Capsule, length=10, radius=5, shapeName=dyn_blouse_fr_02_r_capsule, center=<2.8283 0 5>, rotation=<rot 0 0 90>}]
 dyn_blouse_fr_03_l [{shapeType=Capsule, length=10, radius=6, shapeName=dyn_blouse_fr_03_l_capsule, center=<2.6389 0 3.5>, rotation=<rot 0 0 85>}]
 dyn_blouse_fr_03_r [{shapeType=Capsule, length=10, radius=6, shapeName=dyn_blouse_fr_03_r_capsule, center=<2.6389 0 4>, rotation=<rot 0 0 90>}]
 dyn_blouse_fr_04_l [{
   shapeType=Capsule
   length=18.5669
   radius=6
   shapeName=dyn_blouse_fr_04_l_capsule
   center=<3.8362 0.1062 3.5>
   rotation=<rot 0.0001 0.0001 84.6183>
  }]
 dyn_blouse_fr_04_r [{shapeType=Capsule, length=10, radius=6, shapeName=dyn_blouse_fr_04_r_capsule, center=<5 -0.0728 2.5>, rotation=<rot 0 0 101.3898>}]
 dyn_blouse_side_01_l [{shapeType=Capsule, length=10, radius=6, shapeName=dyn_blouse_side_01_l_capsule, center=<0 0 3>, rotation=<rot 0 0 90>}]
 dyn_blouse_side_01_r [{shapeType=Capsule, length=10, radius=6, shapeName=dyn_blouse_side_01_r_capsule, center=<0 0 4>, rotation=<rot 0 0 90>}]
 dyn_blouse_side_02_l [{shapeType=Capsule, length=10, radius=6, shapeName=dyn_blouse_side_02_l_capsule, center=<1.0744 0 3>, rotation=<rot 0 0 90>}]
 dyn_blouse_side_02_r [{shapeType=Capsule, length=10, radius=6, shapeName=dyn_blouse_side_02_r_capsule, center=<2.3363 0 4>, rotation=<rot 0 0 90>}]
 dyn_blouse_side_03_l [{shapeType=Capsule, length=10, radius=6, shapeName=dyn_blouse_side_03_l_capsule, center=<1.9606 0 4.5>, rotation=<rot 0 0 90>}]
 dyn_blouse_side_03_r [{shapeType=Capsule, length=10, radius=6, shapeName=dyn_blouse_side_03_r_capsule, center=<3.3926 0 4>, rotation=<rot 0 0 90>}]
 dyn_blouse_side_04_l [{shapeType=Capsule, length=22.2812, radius=8, shapeName=dyn_blouse_side_04_l_capsule, center=<2.4002 0.0014 3>, rotation=<rot 0 0 -90>}]
 dyn_blouse_side_04_r [{shapeType=Capsule, length=10, radius=6, shapeName=dyn_blouse_side_04_r_capsule, center=<5 0 3>, rotation=<rot 0 0 94.0259>}]
 pelvis [{shapeType=Capsule, length=7.2018, radius=11.1647, shapeName=pelvis_capsule, center=<-4.1998 -1.6032 0>}]
 spine_01 [{shapeType=Capsule, length=9.7118, radius=8.995, shapeName=spine_01_capsule, center=<-1.7079 -1.2288 0>}]
 spine_02 [{shapeType=Capsule, length=9.214, radius=7.9634, shapeName=spine_02_capsule, center=<1.5207 -0.8764 0>}]
 thigh_twist_01_l [
  {
   shapeType=Capsule
   length=11.8265
   radius=8.6496
   shapeName=thigh_twist_01_l_capsule
   center=<16.3752 -0.1878 1.2554>
   rotation=<rot 85.6565 179.9999 180>
  }
  {
   shapeType=Capsule
   length=20.8168
   radius=7.1229
   shapeName=thigh_twist_01_l_capsule
   center=<4.2168 0 0.6701>
   rotation=<rot 89.3345 0.0002 0.0002>
  }
 ]
 thigh_twist_01_r [
  {
   shapeType=Capsule
   length=12.1852
   radius=8.4974
   shapeName=thigh_twist_01_r_capsule
   center=<-16.3877 0.1878 -1.0698>
   rotation=<rot 86.1401 179.9999 179.9999>
  }
  {
   shapeType=Capsule
   length=20.447
   radius=7.4291
   shapeName=thigh_twist_01_r_capsule
   center=<-4.7479 -0.6499 -0.8499>
   rotation=<rot 89.3925 179.9993 179.9994>
  }
 ]
}
constraints [
 {
  bone1Name=spine_01
  bone2Name=pelvis
  swing1LimitDegrees=45
  swing1Motion=Free
  swing2LimitDegrees=45
  swing2Motion=Free
  twistLimitDegrees=45
  twistMotion=Free
 }
 {
  bone1Name=spine_02
  bone2Name=spine_01
  swing1LimitDegrees=45
  swing1Motion=Free
  swing2LimitDegrees=45
  swing2Motion=Free
  twistLimitDegrees=45
  twistMotion=Free
 }
 {
  bone1Name=dyn_blouse_base_01_c
  bone2Name=spine_02
  swing1LimitDegrees=45
  swing1Motion=Locked
  swing2LimitDegrees=45
  swing2Motion=Locked
  twistLimitDegrees=45
  twistMotion=Locked
 }
 {bone1Name=dyn_blouse_01_c, bone2Name=dyn_blouse_base_01_c, swing1LimitDegrees=5.76, swing2LimitDegrees=25, twistLimitDegrees=2}
 {bone1Name=dyn_blouse_02_c, bone2Name=dyn_blouse_01_c, swing1LimitDegrees=11.08, swing2LimitDegrees=25, twistLimitDegrees=2}
 {bone1Name=dyn_blouse_03_c, bone2Name=dyn_blouse_02_c, swing1LimitDegrees=3.68, swing2LimitDegrees=25, twistLimitDegrees=2}
 {bone1Name=dyn_blouse_04_c, bone2Name=dyn_blouse_03_c, swing1LimitDegrees=7.12, swing2LimitDegrees=25, twistLimitDegrees=2}
 {
  bone1Name=dyn_blouse_base_bk_l
  bone2Name=spine_02
  swing1LimitDegrees=45
  swing1Motion=Free
  swing2LimitDegrees=45
  swing2Motion=Free
  twistLimitDegrees=45
  twistMotion=Free
 }
 {bone1Name=dyn_blouse_bk_01_l, bone2Name=dyn_blouse_base_bk_l, swing1LimitDegrees=25, swing2LimitDegrees=25, twistLimitDegrees=2}
 {bone1Name=dyn_blouse_bk_02_l, bone2Name=dyn_blouse_bk_01_l, swing1LimitDegrees=25, swing2LimitDegrees=25, twistLimitDegrees=2}
 {bone1Name=dyn_blouse_bk_03_l, bone2Name=dyn_blouse_bk_02_l, swing1LimitDegrees=25, swing2LimitDegrees=30, twistLimitDegrees=0}
 {bone1Name=dyn_blouse_bk_04_l, bone2Name=dyn_blouse_bk_03_l, swing1LimitDegrees=25, swing2LimitDegrees=25, twistLimitDegrees=2}
 {
  bone1Name=dyn_blouse_base_bk_r
  bone2Name=spine_02
  swing1LimitDegrees=45
  swing1Motion=Free
  swing2LimitDegrees=45
  swing2Motion=Free
  twistLimitDegrees=45
  twistMotion=Free
 }
 {bone1Name=dyn_blouse_bk_01_r, bone2Name=dyn_blouse_base_bk_r, swing1LimitDegrees=25, swing2LimitDegrees=25, twistLimitDegrees=2}
 {bone1Name=dyn_blouse_bk_02_r, bone2Name=dyn_blouse_bk_01_r, swing1LimitDegrees=25, swing2LimitDegrees=25, twistLimitDegrees=2}
 {bone1Name=dyn_blouse_bk_03_r, bone2Name=dyn_blouse_bk_02_r, swing1LimitDegrees=25, swing2LimitDegrees=25, twistLimitDegrees=2}
 {bone1Name=dyn_blouse_bk_04_r, bone2Name=dyn_blouse_bk_03_r, swing1LimitDegrees=25, swing2LimitDegrees=25, twistLimitDegrees=2}
 {
  bone1Name=dyn_blouse_base_fr_l
  bone2Name=spine_02
  swing1LimitDegrees=45
  swing1Motion=Free
  swing2LimitDegrees=45
  swing2Motion=Free
  twistLimitDegrees=45
  twistMotion=Free
 }
 {bone1Name=dyn_blouse_fr_01_l, bone2Name=dyn_blouse_base_fr_l, swing1LimitDegrees=20, swing2LimitDegrees=35, twistLimitDegrees=2}
 {bone1Name=dyn_blouse_fr_02_l, bone2Name=dyn_blouse_fr_01_l, swing1LimitDegrees=20, swing2LimitDegrees=35, twistLimitDegrees=2}
 {bone1Name=dyn_blouse_fr_03_l, bone2Name=dyn_blouse_fr_02_l, swing1LimitDegrees=15, swing2LimitDegrees=35, twistLimitDegrees=2}
 {bone1Name=dyn_blouse_fr_04_l, bone2Name=dyn_blouse_fr_03_l, swing1LimitDegrees=15, swing2LimitDegrees=40, twistLimitDegrees=15}
 {
  bone1Name=dyn_blouse_base_fr_r
  bone2Name=spine_02
  swing1LimitDegrees=45
  swing1Motion=Free
  swing2LimitDegrees=45
  swing2Motion=Free
  twistLimitDegrees=45
  twistMotion=Free
 }
 {bone1Name=dyn_blouse_fr_01_r, bone2Name=dyn_blouse_base_fr_r, swing1LimitDegrees=20, swing2LimitDegrees=35, twistLimitDegrees=2}
 {bone1Name=dyn_blouse_fr_02_r, bone2Name=dyn_blouse_fr_01_r, swing1LimitDegrees=20, swing2LimitDegrees=35, twistLimitDegrees=2}
 {bone1Name=dyn_blouse_fr_03_r, bone2Name=dyn_blouse_fr_02_r, swing1LimitDegrees=15, swing2LimitDegrees=35, twistLimitDegrees=2}
 {bone1Name=dyn_blouse_fr_04_r, bone2Name=dyn_blouse_fr_03_r, swing1LimitDegrees=15, swing2LimitDegrees=40, twistLimitDegrees=2}
 {
  bone1Name=dyn_blouse_base_side_l
  bone2Name=spine_02
  swing1LimitDegrees=45
  swing1Motion=Free
  swing2LimitDegrees=45
  swing2Motion=Free
  twistLimitDegrees=45
  twistMotion=Free
 }
 {bone1Name=dyn_blouse_side_01_l, bone2Name=dyn_blouse_base_side_l, swing1LimitDegrees=20, swing2LimitDegrees=35, twistLimitDegrees=2}
 {bone1Name=dyn_blouse_side_02_l, bone2Name=dyn_blouse_side_01_l, swing1LimitDegrees=20, swing2LimitDegrees=35, twistLimitDegrees=2}
 {bone1Name=dyn_blouse_side_03_l, bone2Name=dyn_blouse_side_02_l, swing1LimitDegrees=15, swing2LimitDegrees=35, twistLimitDegrees=2}
 {bone1Name=dyn_blouse_side_04_l, bone2Name=dyn_blouse_side_03_l, swing1LimitDegrees=15, swing2LimitDegrees=35, twistLimitDegrees=5}
 {bone1Name=dyn_blouse_base_side_r, bone2Name=spine_02, swing1LimitDegrees=5, swing2LimitDegrees=35, twistLimitDegrees=2}
 {
  bone1Name=dyn_blouse_side_01_r
  bone2Name=dyn_blouse_base_side_r
  swing1LimitDegrees=45
  swing1Motion=Free
  swing2LimitDegrees=45
  swing2Motion=Free
  twistLimitDegrees=45
  twistMotion=Free
 }
 {bone1Name=dyn_blouse_side_02_r, bone2Name=dyn_blouse_side_01_r, swing1LimitDegrees=5, swing2LimitDegrees=35, twistLimitDegrees=2}
 {bone1Name=dyn_blouse_side_03_r, bone2Name=dyn_blouse_side_02_r, swing1LimitDegrees=5, swing2LimitDegrees=35, twistLimitDegrees=2}
 {bone1Name=dyn_blouse_side_04_r, bone2Name=dyn_blouse_side_03_r, swing1LimitDegrees=5, swing2LimitDegrees=35, twistLimitDegrees=2}
 {
  bone1Name=thigh_twist_01_l
  bone2Name=pelvis
  swing1LimitDegrees=45
  swing1Motion=Free
  swing2LimitDegrees=45
  swing2Motion=Free
  twistLimitDegrees=45
  twistMotion=Free
 }
 {
  bone1Name=thigh_twist_01_r
  bone2Name=pelvis
  swing1LimitDegrees=45
  swing1Motion=Free
  swing2LimitDegrees=45
  swing2Motion=Free
  twistLimitDegrees=45
  twistMotion=Free
 }
 {
  bone1Name=dyn_blouse_side_04_r
  bone2Name=dyn_blouse_bk_04_r
  swing1LimitDegrees=45
  swing1Motion=Locked
  swing2LimitDegrees=45
  swing2Motion=Locked
  twistLimitDegrees=45
  twistMotion=Locked
 }
 {
  bone1Name=dyn_blouse_fr_04_r
  bone2Name=dyn_blouse_side_04_r
  swing1LimitDegrees=45
  swing1Motion=Locked
  swing2LimitDegrees=45
  swing2Motion=Locked
  twistLimitDegrees=45
  twistMotion=Locked
 }
 {
  bone1Name=dyn_blouse_side_04_l
  bone2Name=dyn_blouse_bk_04_l
  swing1LimitDegrees=45
  swing1Motion=Locked
  swing2LimitDegrees=45
  swing2Motion=Locked
  twistLimitDegrees=45
  twistMotion=Locked
 }
 {
  bone1Name=dyn_blouse_fr_04_l
  bone2Name=dyn_blouse_side_04_l
  swing1LimitDegrees=45
  swing1Motion=Locked
  swing2LimitDegrees=45
  swing2Motion=Locked
  twistLimitDegrees=45
  twistMotion=Locked
 }
]
