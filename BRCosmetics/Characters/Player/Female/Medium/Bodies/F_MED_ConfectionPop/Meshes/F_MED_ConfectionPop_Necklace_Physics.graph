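# Physics asset for the necklace of the female medium ConfectionPop character body, covering a swinging pendant and the upper torso.
# Three pendant bodies form a chain hanging from spine_05, with tight, asymmetric swing limits and zero twist at each joint.
# Large spine capsules approximate the torso, with free angular motion between them.
# The last pendant body uses two crossed capsules to approximate a wider shape.
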PhysicsAsset /BRCosmetics/Characters/Player/Female/Medium/Bodies/F_MED_ConfectionPop/Meshes/F_MED_ConfectionPop_Necklace_Physics
bodies {
 dyn_pendent_02 [{shapeType=Capsule, length=4.0844, radius=2, shapeName=dyn_pendent_02_capsule, center=<0 0 1.8168>, rotation=<rot 0 0 -90>}]
 dyn_pendent_03 [{shapeType=Capsule, length=8.1285, radius=2, shapeName=dyn_pendent_03_capsule, center=<0 0 0.332>, rotation=<rot 0 0 -90>}]
 dyn_pendent_05 [
  {shapeType=Capsule, length=11, radius=1.5, shapeName=dyn_pendent_05_capsule, center=<2.9007 0.2106 0.5>, rotation=<rot 90 0 -45>}
  {shapeType=Capsule, length=11, radius=1.5, shapeName=dyn_pendent_05_capsule, center=<2.9007 0.2106 0.5>, rotation=<rot 90 0 45>}
 ]
 spine_04 [{
   shapeType=Capsule
   length=6.805
   radius=10.2657
   shapeName=spine_04_capsule
   center=<-0.3202 1.2977 0.1523>
   rotation=<rot -5.6362 -90 0.8622>
  }]
 spine_05 [{
   shapeType=Capsule
   length=4.9363
   radius=12.3323
   shapeName=spine_05_capsule
   center=<-3.3911 -2.7244 -0.3459>
   rotation=<rot 0.0233 95.8551 179.9837>
  }]
}
constraints [
 {
  bone1Name=spine_05
  bone2Name=spine_04
  swing1LimitDegrees=45
  swing1Motion=Free
  swing2LimitDegrees=45
  swing2Motion=Free
  twistLimitDegrees=45
  twistMotion=Free
 }
 {bone1Name=dyn_pendent_02, bone2Name=spine_05, swing1LimitDegrees=5, swing2LimitDegrees=15, twistLimitDegrees=0}
 {bone1Name=dyn_pendent_03, bone2Name=dyn_pendent_02, swing1LimitDegrees=2, swing2LimitDegrees=20, twistLimitDegrees=0}
 {bone1Name=dyn_pendent_05, bone2Name=dyn_pendent_03, swing1LimitDegrees=1, swing2LimitDegrees=15, twistLimitDegrees=0}
]
