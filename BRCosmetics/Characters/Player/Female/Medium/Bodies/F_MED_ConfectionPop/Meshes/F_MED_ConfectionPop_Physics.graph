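# Physics asset for the ConfectionPop medium female body, rigging its blouse for secondary motion around capsules on pelvis, lower spine and thighs.
# Seven capsule chains (center, plus left/right front, side and back) hang from base bodies jointed to spine_02.
# Chain links have tight per-region swing limits and little or no twist.
# Angularly locked cross-constraints tie neighboring chain tips into a hem that stays open between the two front chains.
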PhysicsAsset /BRCosmetics/Characters/Player/Female/Medium/Bodies/F_MED_ConfectionPop/Meshes/F_MED_ConfectionPop_Physics
bodies {
 dyn_blouse_01_c [{shapeType=Capsule, length=10, radius=5, shapeName=dyn_blouse_01_c_capsule, center=<0 0 3>, rotation=<rot 0 0 90>}]
 dyn_blouse_02_c [{shapeType=Capsule, length=10, radius=5, shapeName=dyn_blouse_02_c_capsule, center=<0 0 3>, rotation=<rot 0 0 90>}]
 dyn_blouse_03_c [{shapeType=Capsule, length=7.8818, radius=7.1182, shapeName=dyn_blouse_03_c_capsule, center=<0 0 3>, rotation=<rot 0 0 90>}]
 dyn_blouse_04_c [{shapeType=Capsule, length=11.3434, radius=9.3319, shapeName=dyn_blouse_04_c_capsule, center=<2.1305 0 3>, rotation=<rot 0 0 90>}]
 dyn_blouse_base_01_c [{shapeType=Capsule, length=10, radius=5, shapeName=dyn_blouse_base_01_c_capsule, rotation=<rot 0 0 90>}]
 dyn_blouse_base_bk_l [{shapeType=Capsule, length=10, radius=5, shapeName=dyn_blouse_base_bk_l_capsule, rotation=<rot 0 0 90>}]
 dyn_blouse_base_bk_r [{shapeType=Capsule, length=10, radius=5, shapeName=dyn_blouse_base_bk_r_capsule, rotation=<rot 0 0 90>}]
 dyn_blouse_base_fr_l [{shapeType=Capsule, length=10, radius=5, shapeName=dyn_blouse_base_fr_l_capsule, rotation=<rot 0 0 90>}]
 dyn_blouse_base_fr_r [{shapeType=Capsule, length=10, radius=5, shapeName=dyn_blouse_base_fr_r_capsule, rotation=<rot 0 0 90>}]
 dyn_blouse_base_side_l [{shapeType=Capsule, length=10, radius=5, shapeName=dyn_blouse_base_side_l_capsule, rotation=<rot 0 0 90>}]
 dyn_blouse_base_side_r [{shapeType=Capsule, length=10, radius=5, shapeName=dyn_blouse_base_side_r_capsule, rotation=<rot 0 0 90>}]
 dyn_blouse_bk_01_l [{shapeType=Capsule, length=10, radius=5, shapeName=dyn_blouse_bk_01_l_capsule, center=<0 0 3>, rotation=<rot 0 0 90>}]
 dyn_blouse_bk_01_r [{shapeType=Capsule, length=10, radius=5, shapeName=dyn_blouse_bk_01_r_capsule, center=<0 0 3>, rotation=<rot 0 0 90>}]
 dyn_blouse_bk_02_l [{shapeType=Capsule, length=10, radius=5, shapeName=dyn_blouse_bk_02_l_capsule, center=<0 0 3>, rotation=<rot 0 0 90>}]
 dyn_blouse_bk_02_r [{shapeType=Capsule, length=10, radius=5, shapeName=dyn_blouse_bk_02_r_capsule, center=<0 0 3>, rotation=<rot 0 0 90>}]
 dyn_blouse_bk_03_l [{shapeType=Capsule, length=7.8818, radius=7.1182, shapeName=dyn_blouse_bk_03_l_capsule, center=<0 0 3>, rotation=<rot 0 0 90>}]
 dyn_blouse_bk_03_r [{shapeType=Capsule, length=7.8818, radius=7.1182, shapeName=dyn_blouse_bk_03_r_capsule, center=<0 0 3>, rotation=<rot 0 0 90>}]
 dyn_blouse_bk_04_l [{shapeType=Capsule, length=13.6471, radius=8.9187, shapeName=dyn_blouse_bk_04_l_capsule, center=<2.2538 0 3>, rotation=<rot 0 0 95>}]
 dyn_blouse_bk_04_r [{shapeType=Capsule, length=13.0517, radius=9.6013, shapeName=dyn_blouse_bk_04_r_capsule, center=<2.322 0 3>, rotation=<rot 0 0 95>}]
 dyn_blouse_fr_01_l [{shapeType=Capsule, length=10, radius=5, shapeName=dyn_blouse_fr_01_l_capsule, center=<0 0 3>, rotation=<rot 0 0 90>}]
 dyn_blouse_fr_01_r [{shapeType=Capsule, length=10, radius=5, shapeName=dyn_blouse_fr_01_r_capsule, center=<0 0 3>, rotation=<rot 0 0 90>}]
 dyn_blouse_fr_02_l [{shapeType=Capsule, length=10, radius=5, shapeName=dyn_blouse_fr_02_l_capsule, center=<0 0 3>, rotation=<rot 0 0 90>}]
 dyn_blouse_fr_02_r [{shapeType=Capsule, length=10, radius=5, shapeName=dyn_blouse_fr_02_r_capsule, center=<0 0 3>, rotation=<rot 0 0 90>}]
 dyn_blouse_fr_03_l [{shapeType=Capsule, length=7.8818, radius=7.1182, shapeName=dyn_blouse_fr_03_l_capsule, center=<0 0 5>, rotation=<rot 0 0 90>}]
 dyn_blouse_fr_03_r [{shapeType=Capsule, length=7.8818, radius=7.1182, shapeName=dyn_blouse_fr_03_r_capsule, center=<0 0 5>, rotation=<rot 0 0 90>}]
 dyn_blouse_fr_04_l [{shapeType=Capsule, length=10, radius=5.512, shapeName=dyn_blouse_fr_04_l_capsule, center=<2.7143 0 5>, rotation=<rot 0 0 90>}]
 dyn_blouse_fr_04_r [{shapeType=Capsule, length=8.5469, radius=7.056, shapeName=dyn_blouse_fr_04_r_capsule, center=<2.5176 -0.1167 5>, rotation=<rot 0 0 90>}]
 dyn_blouse_side_01_l [{shapeType=Capsule, length=10, radius=5, shapeName=dyn_blouse_side_01_l_capsule, center=<0 0 3>, rotation=<rot 0 0 90>}]
 dyn_blouse_side_01_r [{shapeType=Capsule, length=10, radius=5, shapeName=dyn_blouse_side_01_r_capsule, center=<0 0 3>, rotation=<rot 0 0 90>}]
 dyn_blouse_side_02_l [{shapeType=Capsule, length=10, radius=5, shapeName=dyn_blouse_side_02_l_capsule, center=<0 0 3>, rotation=<rot 0 0 90>}]
 dyn_blouse_side_02_r [{shapeType=Capsule, length=10, radius=5, shapeName=dyn_blouse_side_02_r_capsule, center=<0 0 3>, rotation=<rot 0 0 90>}]
 dyn_blouse_side_03_l [{shapeType=Capsule, length=7.8818, radius=7.1182, shapeName=dyn_blouse_side_03_l_capsule, center=<0 0 4>, rotation=<rot 0 0 90>}]
 dyn_blouse_side_03_r [{shapeType=Capsule, length=7.8818, radius=7.1182, shapeName=dyn_blouse_side_03_r_capsule, center=<0 0 4>, rotation=<rot 0 0 90>}]
 dyn_blouse_side_04_l [{shapeType=Capsule, length=8.5469, radius=8.7263, shapeName=dyn_blouse_side_04_l_capsule, center=<2.284 0 4>, rotation=<rot 0 0 85>}]
 dyn_blouse_side_04_r [{shapeType=Capsule, length=9.7501, radius=9.2947, shapeName=dyn_blouse_side_04_r_capsule, center=<1.9374 0 4>, rotation=<rot 0 0 95>}]
 pelvis [{shapeType=Capsule, length=8.1652, radius=11.1647, shapeName=pelvis_capsule, center=<-1.4429 -0.4526 0>}]
 spine_01 [{shapeType=Capsule, length=8.7957, radius=8.995, shapeName=spine_01_capsule, center=<0.7103 -0.023 0>}]
 spine_02 [{shapeType=Capsule, length=9.9996, radius=7.9634, shapeName=spine_02_capsule, center=<3.2663 0 0>}]
 thigh_l [{
   shapeType=Capsule
   length=29.2565
   radius=8.1892
   shapeName=thigh_l_capsule
   center=<-12.9585 -1.3766 0.4223>
   rotation=<rot 0.1824 87.2778 -88.9656>
  }]
 thigh_r [{
   shapeType=Capsule
   length=24.6725
   radius=9.4413
   shapeName=thigh_r_capsule
   center=<13.2364 0.3997 -2.3368>
   rotation=<rot 0.2971 -92.6036 -93.8913>
  }]
}
constraints [
 {
  bone1Name=spine_01
  bone2Name=pelvis
  swing1LimitDegrees=45
  swing1Motion=Free
  swing2LimitDegrees=45
  swing2Motion=Free
  twistLimitDegrees=45
  twistMotion=Free
 }
 {
  bone1Name=spine_02
  bone2Name=spine_01
  swing1LimitDegrees=45
  swing1Motion=Free
  swing2LimitDegrees=45
  swing2Motion=Free
  twistLimitDegrees=45
  twistMotion=Free
 }
 {
  bone1Name=thigh_l
  bone2Name=pelvis
  swing1LimitDegrees=45
  swing1Motion=Free
  swing2LimitDegrees=45
  swing2Motion=Free
  twistLimitDegrees=45
  twistMotion=Free
 }
 {
  bone1Name=thigh_r
  bone2Name=pelvis
  swing1LimitDegrees=45
  swing1Motion=Free
  swing2LimitDegrees=45
  swing2Motion=Free
  twistLimitDegrees=45
  twistMotion=Free
 }
 {
  bone1Name=dyn_blouse_base_bk_l
  bone2Name=spine_02
  swing1LimitDegrees=45
  swing1Motion=Free
  swing2LimitDegrees=45
  swing2Motion=Free
  twistLimitDegrees=45
  twistMotion=Free
 }
 {bone1Name=dyn_blouse_bk_01_l, bone2Name=dyn_blouse_base_bk_l, swing1LimitDegrees=10, swing2LimitDegrees=15, twistLimitDegrees=0}
 {bone1Name=dyn_blouse_bk_02_l, bone2Name=dyn_blouse_bk_01_l, swing1LimitDegrees=10, swing2LimitDegrees=15, twistLimitDegrees=0}
 {bone1Name=dyn_blouse_bk_03_l, bone2Name=dyn_blouse_bk_02_l, swing1LimitDegrees=10, swing2LimitDegrees=15, twistLimitDegrees=0}
 {bone1Name=dyn_blouse_bk_04_l, bone2Name=dyn_blouse_bk_03_l, swing1LimitDegrees=10, swing2LimitDegrees=15, twistLimitDegrees=0}
 {
  bone1Name=dyn_blouse_base_bk_r
  bone2Name=spine_02
  swing1LimitDegrees=45
  swing1Motion=Free
  swing2LimitDegrees=45
  swing2Motion=Free
  twistLimitDegrees=45
  twistMotion=Free
 }
 {bone1Name=dyn_blouse_bk_01_r, bone2Name=dyn_blouse_base_bk_r, swing1LimitDegrees=10, swing2LimitDegrees=15, twistLimitDegrees=0}
 {bone1Name=dyn_blouse_bk_02_r, bone2Name=dyn_blouse_bk_01_r, swing1LimitDegrees=10, swing2LimitDegrees=15, twistLimitDegrees=0}
 {bone1Name=dyn_blouse_bk_03_r, bone2Name=dyn_blouse_bk_02_r, swing1LimitDegrees=10, swing2LimitDegrees=15, twistLimitDegrees=2}
 {bone1Name=dyn_blouse_bk_04_r, bone2Name=dyn_blouse_bk_03_r, swing1LimitDegrees=10, swing2LimitDegrees=15, twistLimitDegrees=2}
 {
  bone1Name=dyn_blouse_base_fr_l
  bone2Name=spine_02
  swing1LimitDegrees=45
  swing1Motion=Free
  swing2LimitDegrees=45
  swing2Motion=Free
  twistLimitDegrees=45
  twistMotion=Free
 }
 {bone1Name=dyn_blouse_fr_01_l, bone2Name=dyn_blouse_base_fr_l, swing1LimitDegrees=30, swing2LimitDegrees=10, twistLimitDegrees=0}
 {bone1Name=dyn_blouse_fr_02_l, bone2Name=dyn_blouse_fr_01_l, swing1LimitDegrees=30, swing2LimitDegrees=10, twistLimitDegrees=0}
 {bone1Name=dyn_blouse_fr_03_l, bone2Name=dyn_blouse_fr_02_l, swing1LimitDegrees=30, swing2LimitDegrees=10, twistLimitDegrees=0}
 {bone1Name=dyn_blouse_fr_04_l, bone2Name=dyn_blouse_fr_03_l, swing1LimitDegrees=30, swing2LimitDegrees=10, twistLimitDegrees=0}
 {
  bone1Name=dyn_blouse_base_fr_r
  bone2Name=spine_02
  swing1LimitDegrees=45
  swing1Motion=Free
  swing2LimitDegrees=45
  swing2Motion=Free
  twistLimitDegrees=45
  twistMotion=Free
 }
 {bone1Name=dyn_blouse_fr_01_r, bone2Name=dyn_blouse_base_fr_r, swing1LimitDegrees=30, swing2LimitDegrees=10, twistLimitDegrees=0}
 {bone1Name=dyn_blouse_fr_02_r, bone2Name=dyn_blouse_fr_01_r, swing1LimitDegrees=30, swing2LimitDegrees=10, twistLimitDegrees=0}
 {bone1Name=dyn_blouse_fr_03_r, bone2Name=dyn_blouse_fr_02_r, swing1LimitDegrees=30, swing2LimitDegrees=10, twistLimitDegrees=0}
 {bone1Name=dyn_blouse_fr_04_r, bone2Name=dyn_blouse_fr_03_r, swing1LimitDegrees=30, swing2LimitDegrees=10, twistLimitDegrees=0}
 {
  bone1Name=dyn_blouse_base_side_l
  bone2Name=spine_02
  swing1LimitDegrees=45
  swing1Motion=Free
  swing2LimitDegrees=45
  swing2Motion=Free
  twistLimitDegrees=45
  twistMotion=Free
 }
 {bone1Name=dyn_blouse_side_01_l, bone2Name=dyn_blouse_base_side_l, swing1LimitDegrees=15, swing2LimitDegrees=15, twistLimitDegrees=0}
 {bone1Name=dyn_blouse_side_02_l, bone2Name=dyn_blouse_side_01_l, swing1LimitDegrees=15, swing2LimitDegrees=15, twistLimitDegrees=0}
 {bone1Name=dyn_blouse_side_03_l, bone2Name=dyn_blouse_side_02_l, swing1LimitDegrees=15, swing2LimitDegrees=15, twistLimitDegrees=0}
 {bone1Name=dyn_blouse_side_04_l, bone2Name=dyn_blouse_side_03_l, swing1LimitDegrees=15, swing2LimitDegrees=15, twistLimitDegrees=0}
 {
  bone1Name=dyn_blouse_base_side_r
  bone2Name=spine_02
  swing1LimitDegrees=45
  swing1Motion=Free
  swing2LimitDegrees=45
  swing2Motion=Free
  twistLimitDegrees=45
  twistMotion=Free
 }
 {bone1Name=dyn_blouse_side_01_r, bone2Name=dyn_blouse_base_side_r, swing1LimitDegrees=15, swing2LimitDegrees=15, twistLimitDegrees=0}
 {bone1Name=dyn_blouse_side_02_r, bone2Name=dyn_blouse_side_01_r, swing1LimitDegrees=15, swing2LimitDegrees=15, twistLimitDegrees=0}
 {bone1Name=dyn_blouse_side_03_r, bone2Name=dyn_blouse_side_02_r, swing1LimitDegrees=15, swing2LimitDegrees=15, twistLimitDegrees=0}
 {bone1Name=dyn_blouse_side_04_r, bone2Name=dyn_blouse_side_03_r, swing1LimitDegrees=15, swing2LimitDegrees=15, twistLimitDegrees=0}
 {bone1Name=dyn_blouse_04_c, bone2Name=dyn_blouse_03_c, swing1LimitDegrees=5, swing2LimitDegrees=15, twistLimitDegrees=0}
 {
  bone1Name=dyn_blouse_base_01_c
  bone2Name=spine_02
  swing1LimitDegrees=45
  swing1Motion=Free
  swing2LimitDegrees=45
  swing2Motion=Free
  twistLimitDegrees=45
  twistMotion=Free
 }
 {bone1Name=dyn_blouse_01_c, bone2Name=dyn_blouse_base_01_c, swing1LimitDegrees=0, swing2LimitDegrees=5, twistLimitDegrees=0}
 {bone1Name=dyn_blouse_02_c, bone2Name=dyn_blouse_01_c, swing1LimitDegrees=0, swing2LimitDegrees=15, twistLimitDegrees=0}
 {bone1Name=dyn_blouse_03_c, bone2Name=dyn_blouse_02_c, swing1LimitDegrees=5, swing2LimitDegrees=15, twistLimitDegrees=0}
 {
  bone1Name=dyn_blouse_04_c
  bone2Name=dyn_blouse_bk_04_r
  swing1LimitDegrees=45
  swing1Motion=Locked
  swing2LimitDegrees=45
  swing2Motion=Locked
  twistLimitDegrees=45
  twistMotion=Locked
 }
 {
  bone1Name=dyn_blouse_04_c
  bone2Name=dyn_blouse_bk_04_l
  swing1LimitDegrees=45
  swing1Motion=Locked
  swing2LimitDegrees=45
  swing2Motion=Locked
  twistLimitDegrees=45
  twistMotion=Locked
 }
 {
  bone1Name=dyn_blouse_side_04_l
  bone2Name=dyn_blouse_fr_04_l
  swing1LimitDegrees=45
  swing1Motion=Locked
  swing2LimitDegrees=45
  swing2Motion=Locked
  twistLimitDegrees=45
  twistMotion=Locked
 }
 {
  bone1Name=dyn_blouse_side_04_r
  bone2Name=dyn_blouse_fr_04_r
  swing1LimitDegrees=45
  swing1Motion=Locked
  swing2LimitDegrees=45
  swing2Motion=Locked
  twistLimitDegrees=45
  twistMotion=Locked
 }
 {
  bone1Name=dyn_blouse_bk_04_r
  bone2Name=dyn_blouse_side_04_r
  swing1LimitDegrees=45
  swing1Motion=Locked
  swing2LimitDegrees=45
  swing2Motion=Locked
  twistLimitDegrees=45
  twistMotion=Locked
 }
 {
  bone1Name=dyn_blouse_bk_04_l
  bone2Name=dyn_blouse_side_04_l
  swing1LimitDegrees=45
  swing1Motion=Locked
  swing2LimitDegrees=45
  swing2Motion=Locked
  twistLimitDegrees=45
  twistMotion=Locked
 }
]
