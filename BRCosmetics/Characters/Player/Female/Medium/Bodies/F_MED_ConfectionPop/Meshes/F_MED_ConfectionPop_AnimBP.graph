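# Post-process style animation Blueprint for the ConfectionPop female body cosmetic, adding secondary motion to a pose copied from the attached parent mesh.
# Control Rigs and bone offsets feed a cached pose, then skydive, emote and front-end flags select among RigidBody simulations.
# Necklace physics, slide-driven blouse offsets and AnimDynamics laces follow.
# The event graph drives material and overlay parameters for WPO cloth motion and upper-arm pose blends, computed from rotation dot products.
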AnimBlueprint /BRCosmetics/Characters/Player/Female/Medium/Bodies/F_MED_ConfectionPop/Meshes/F_MED_ConfectionPop_AnimBP
target_skeleton=&/F_MED_ConfectionPop_Skeleton
graphs [
 AnimGraph {nodes [
   n1 = "Constraint - dyn_blouse_base_fr_l (TT)"!(ComponentPose, Alpha=1, ConstraintWeights_0=0.9, ConstraintWeights_1:float=0.1) -> (Pose) {
    alphaBoolBlend {blendOption=Linear}
    boneToModify {boneName=dyn_blouse_base_fr_l}
    constraintSetup [{transformType=Translation, targetBone {boneName=spine_03}}, {transformType=Translation, targetBone {boneName=clavicle_l}}]
    constraintWeights [1,1]
   }
   n2 = "Constraint - dyn_blouse_base_side_l (TT)"!(ComponentPose=@n1, Alpha=1, ConstraintWeights_0=0.8, ConstraintWeights_1:float=0.2) -> (Pose) {
    alphaBoolBlend {blendOption=Linear}
    boneToModify {boneName=dyn_blouse_base_side_l}
    constraintSetup [{transformType=Translation, targetBone {boneName=spine_03}}, {transformType=Translation, targetBone {boneName=clavicle_l}}]
    constraintWeights [1,1]
   }
   n3 = "Constraint - dyn_blouse_base_fr_r (TT)"!(ComponentPose=@n2, Alpha=1, ConstraintWeights_0=0.9, ConstraintWeights_1:float=0.1) -> (Pose) {
    alphaBoolBlend {blendOption=Linear}
    boneToModify {boneName=dyn_blouse_base_fr_r}
    constraintSetup [{transformType=Translation, targetBone {boneName=spine_03}}, {transformType=Translation, targetBone {boneName=clavicle_r}}]
    constraintWeights [1,1]
   }
   "Constraint - dyn_blouse_base_side_r (TT)"!(ComponentPose=@n3, Alpha=1, ConstraintWeights_0=0.8, ConstraintWeights_1:float=0.2) -> (Pose) {
    alphaBoolBlend {blendOption=Linear}
    boneToModify {boneName=dyn_blouse_base_side_r}
    constraintSetup [{transformType=Translation, targetBone {boneName=spine_03}}, {transformType=Translation, targetBone {boneName=clavicle_r}}]
    constraintWeights [1,1]
   }
   n4 = "Transform (Modify) Bone - Bone: upperarm_r"(Alpha=1, Translation="0,0,0", Rotation=Get UpperArmRRot() -> (UpperArmRRot:Rotator), Scale="1,1,1") {rotationMode=BMM_Additive, rotationSpace=BCS_ParentBoneSpace, alphaBoolBlend {blendOption=Linear}, boneToModify {boneName=upperarm_r}}
   "Transform (Modify) Bone - Bone: upperarm_l"(ComponentPose=@n4, Alpha=1, Translation="0,0,0", Rotation=Get UpperArmLRot() -> (UpperArmLRot:Rotator), Scale="1,1,1") {rotationMode=BMM_Additive, rotationSpace=BCS_ParentBoneSpace, alphaBoolBlend {blendOption=Linear}, boneToModify {boneName=upperarm_l}}
   n5 = RigidBody!(
    ComponentPose=Local To Component(LocalPose=Use cached pose 'Pre sim'())
    bAlphaBoolEnabled:bool
    OverrideWorldGravity:Vector="0,0,0"
    ExternalForce
   ) -> (Pose) {
    alphaInputType=Bool
    bClampLinearTranslationLimitToRefPose=true
    bOverrideWorldGravity=true
    overridePhysicsAsset=&/F_MED_ConfectionPop_FrontEdn_Physics
    simulationSpace=BaseBoneSpace
    alphaBoolBlend {blendOption=Linear}
    baseBoneRef {boneName=pelvis}
    componentAppliedLinearAccClamp=<1200 1200 1200>
    componentLinearAccScale=<0.25 0.25 0.4>
    componentLinearVelScale=<0.3 0.3 0.3>
    simSpaceSettings {
     maxAngularAcceleration=100
     maxAngularVelocity=2
     maxLinearAcceleration=100
     maxLinearVelocity=160
     velocityScaleZ=0
     worldAlpha=0.75
     externalLinearDragV=<1 1 1>
    }
   }
   n6 = RigidBody(ComponentPose=Local To Component(LocalPose=Use cached pose 'Pre sim'()), OverrideWorldGravity="0,0,0") {
    alphaInputType=Bool
    bClampLinearTranslationLimitToRefPose=true
    bOverrideWorldGravity=true
    overridePhysicsAsset=&/F_MED_ConfectionPop_DualWield_Physics
    simulationSpace=BaseBoneSpace
    alphaBoolBlend {blendOption=Linear}
    baseBoneRef {boneName=spine_01}
    componentAppliedLinearAccClamp=<2500 2500 2500>
    componentLinearAccScale=<0.25 0.25 0.1>
    componentLinearVelScale=<0.3 0.3 0.1>
    simSpaceSettings {maxAngularAcceleration=100, maxAngularVelocity=3, maxLinearAcceleration=100, maxLinearVelocity=160, velocityScaleZ=0, worldAlpha=1}
   }
   n7 = Control Rig(Source=Component To Local(ComponentPose=@n6)) {
    alphaBoolBlend {blendOption=Linear}
    controlRigAssetReference {blueprintRigClass=&/F_MED_ConfectionPop_Corrective_CtrlRig.F_MED_ConfectionPop_Corrective_CtrlRig_C}
   }
   n8 = RigidBody(ComponentPose=Local To Component(LocalPose=Use cached pose 'Pre sim'()), OverrideWorldGravity="0,0,0") {
    alphaInputType=Bool
    bClampLinearTranslationLimitToRefPose=true
    bOverrideWorldGravity=true
    overridePhysicsAsset=&/F_MED_ConfectionPop_Backpack_Physics
    simulationSpace=BaseBoneSpace
    alphaBoolBlend {blendOption=Linear}
    baseBoneRef {boneName=pelvis}
    componentAppliedLinearAccClamp=<1200 1200 1200>
    componentLinearAccScale=<0.25 0.25 0.4>
    componentLinearVelScale=<0.3 0.3 0.3>
    simSpaceSettings {
     maxAngularAcceleration=100
     maxAngularVelocity=2
     maxLinearAcceleration=100
     maxLinearVelocity=160
     velocityScaleZ=0
     worldAlpha=0.75
     externalLinearDragV=<1 1 1>
    }
   }
   n9 = Get bIsPlayingEmote()
   n10 = Control Rig!(Source=Component To Local(ComponentPose=@n8), Alpha, is Play emote:bool=@n9) -> (Pose) {
    alphaBoolBlend {blendOption=Linear}
    controlRigAssetReference {blueprintRigClass=&F_MED_ConfectionPop_Corrective_CtrlRig_C}
   }
   n11 = Blend Poses by bool(
    BlendPose_0=@n10
    BlendPose_1=Blend Poses by bool(BlendPose_0=Component To Local(ComponentPose=@n5), BlendPose_1=@n7, bActiveValue=Get bIsInFrontEnd()) {blendType=Linear}
    bActiveValue=@n9
   ) {blendType=Linear}
   n12 = RigidBody(ComponentPose=Local To Component(LocalPose=Use cached pose 'Pre sim'()), OverrideWorldGravity="0,0,0") {
    alphaInputType=Bool
    bClampLinearTranslationLimitToRefPose=true
    bOverrideWorldGravity=true
    overridePhysicsAsset=&/F_MED_ConfectionPop_Physics
    simulationSpace=BaseBoneSpace
    alphaBoolBlend {blendOption=Linear}
    baseBoneRef {boneName=pelvis}
    componentAppliedLinearAccClamp=<1200 1200 1200>
    componentLinearAccScale=<0.25 0.25 0.4>
    componentLinearVelScale=<0.3 0.3 0.3>
    simSpaceSettings {
     maxAngularAcceleration=100
     maxAngularVelocity=2
     maxLinearAcceleration=100
     maxLinearVelocity=160
     velocityScaleZ=0
     worldAlpha=0.75
     externalLinearDragV=<1 1 1>
    }
   }
   n13 = Component To Local(ComponentPose=@n12)
   n14 = Get bIsSkyDiving()
   n15 = RigidBody!(
    ComponentPose=Local To Component(LocalPose=Blend Poses by bool(BlendPose_0=@n13, BlendPose_1=@n11, bActiveValue=@n14) {blendType=Linear})
    Alpha:float
    OverrideWorldGravity="0,0,0"
    ExternalForce
   ) -> (Pose) {
    bClampLinearTranslationLimitToRefPose=true
    bOverrideWorldGravity=true
    overridePhysicsAsset=&/F_MED_ConfectionPop_Necklace_Physics
    simulationSpace=BaseBoneSpace
    alphaBoolBlend {blendOption=Linear}
    baseBoneRef {boneName=spine_05}
    componentAppliedLinearAccClamp=<2000 2000 2000>
    componentLinearAccScale=<0.5 0.5 0.5>
    componentLinearVelScale=<0.7 0.7 0.7>
    simSpaceSettings {
     maxAngularAcceleration=200
     maxAngularVelocity=2
     maxLinearAcceleration=150
     maxLinearVelocity=150
     velocityScaleZ=0
     worldAlpha=1
     externalLinearDragV=<1 1 1>
    }
   }
   n16 = Get bIsSliding()
   n17 = "Transform (Modify) Bone - Bone: dyn_blouse_base_fr_l"!(ComponentPose=@n15, bAlphaBoolEnabled:bool=@n16, Translation="5,3,2", Rotation="6,0,0", Scale="1,1,1") -> (Pose) {
    alphaInputType=Bool
    rotationMode=BMM_Additive
    translationMode=BMM_Additive
    alphaBoolBlend {blendOption=Linear}
    boneToModify {boneName=dyn_blouse_base_fr_l}
    rotation {pitch=6}
    translation=<5 3 2>
   }
   n18 = "Transform (Modify) Bone - Bone: dyn_blouse_base_fr_r"!(ComponentPose=@n17, bAlphaBoolEnabled:bool=@n16, Translation="0,0,0", Rotation="0,0,0", Scale="1,1,1") -> (Pose) {
    alphaInputType=Bool
    rotationMode=BMM_Additive
    translationMode=BMM_Additive
    alphaBoolBlend {blendOption=Linear}
    boneToModify {boneName=dyn_blouse_base_fr_r}
   }
   n19 = Anim Dynamics - Bone: dyn_laces_01_l(ComponentPose=@n18, Alpha=1, ExternalForce="0,0,0") {
    angularDampingOverride=0.9999
    angularSpringConstant=250
    bAngularSpring=true
    bOverrideAngularDamping=true
    bUseGravityOverride=true
    simulationSpace=RootRelative
    alphaBoolBlend {blendOption=Linear}
    boundBone {boneName=dyn_laces_01_l}
    componentAppliedLinearAccClamp=<1000 1000 1000>
    componentLinearAccScale=<1 1 1>
    componentLinearVelScale=<1 1 1>
    gravityOverride {z=-920}
    physicsBodyDefinitions [{
      boundBone {boneName=dyn_laces_01_l}
      boxExtents=<2 2 2>
      constraintSetup {angularLimitsMax=<0 30 20>, angularLimitsMin=<0 -30 -3>, angularTarget=<0 2 0>}
      localJointOffset=<0.1 0 0>
     }]
    relativeSpaceBone {boneName=calf_twist_02_r}
   }
   n20 = Anim Dynamics - Bone: dyn_laces_02_l(ComponentPose=@n19, Alpha=1, ExternalForce="0,0,0") {
    angularDampingOverride=0.9999
    angularSpringConstant=250
    bAngularSpring=true
    bOverrideAngularDamping=true
    bUseGravityOverride=true
    simulationSpace=RootRelative
    alphaBoolBlend {blendOption=Linear}
    boundBone {boneName=dyn_laces_02_l}
    componentAppliedLinearAccClamp=<1000 1000 1000>
    componentLinearAccScale=<1 1 1>
    componentLinearVelScale=<1 1 1>
    gravityOverride {z=-920}
    physicsBodyDefinitions [{
      boundBone {boneName=dyn_laces_02_l}
      boxExtents=<2 2 2>
      constraintSetup {angularLimitsMax=<0 30 20>, angularLimitsMin=<0 -30 -3>, angularTarget=<0 2 0>}
      localJointOffset=<0.1 0 0>
     }]
    relativeSpaceBone {boneName=calf_twist_02_r}
   }
   n21 = Anim Dynamics - Bone: dyn_laces_01_r(ComponentPose=@n20, Alpha=1, ExternalForce="0,0,0") {
    angularDampingOverride=0.9999
    angularSpringConstant=250
    bAngularSpring=true
    bOverrideAngularDamping=true
    bUseGravityOverride=true
    simulationSpace=RootRelative
    alphaBoolBlend {blendOption=Linear}
    boundBone {boneName=dyn_laces_01_r}
    componentAppliedLinearAccClamp=<1000 1000 1000>
    componentLinearAccScale=<1 1 1>
    componentLinearVelScale=<1 1 1>
    gravityOverride {z=-920}
    physicsBodyDefinitions [{
      boundBone {boneName=dyn_laces_01_r}
      boxExtents=<2 2 2>
      constraintSetup {angularLimitsMax=<30 5 30>, angularLimitsMin=<-5 -20 -5>, angularTarget=<0 -1 0>}
      localJointOffset=<0.1 0 0>
     }]
    relativeSpaceBone {boneName=calf_twist_02_r}
   }
   n22 = Anim Dynamics - Bone: dyn_laces_02_r(ComponentPose=@n21, Alpha=1, ExternalForce="0,0,0") {
    angularDampingOverride=0.9999
    angularSpringConstant=250
    bAngularSpring=true
    bOverrideAngularDamping=true
    bUseGravityOverride=true
    simulationSpace=RootRelative
    alphaBoolBlend {blendOption=Linear}
    boundBone {boneName=dyn_laces_02_r}
    componentAppliedLinearAccClamp=<1000 1000 1000>
    componentLinearAccScale=<1 1 1>
    componentLinearVelScale=<1 1 1>
    gravityOverride {z=-920}
    physicsBodyDefinitions [{
      boundBone {boneName=dyn_laces_02_r}
      boxExtents=<2 2 2>
      constraintSetup {angularLimitsMax=<5 5 10>, angularLimitsMin=<-20 -20 -20>, angularTarget=<0 0 -0.5>}
      localJointOffset=<0.1 0 0>
     }]
    relativeSpaceBone {boneName=calf_twist_02_r}
   }
   Output Pose(Result=Component To Local(ComponentPose=@n22)) {layerGroup=None}
   n23 = Control Rig!(Source=Copy Pose From Mesh() {bUseAttachedParent=true}, Alpha) -> (Pose) {
    alphaBoolBlend {blendOption=Linear}
    controlRigAssetReference {blueprintRigClass=&/F_MED_ConfectionPop_DeformRig.F_MED_ConfectionPop_DeformRig_C}
   }
   n24 = "Transform (Modify) Bone - Bone: scale_shoulder_l"(ComponentPose=Local To Component(LocalPose=@n23), Alpha=1, Translation="0,0,-10", Rotation="0,0,0", Scale="0,0,0") {
    scaleMode=BMM_Additive
    scaleSpace=BCS_BoneSpace
    translationMode=BMM_Additive
    translationSpace=BCS_BoneSpace
    alphaBoolBlend {blendOption=Linear}
    boneToModify {boneName=scale_shoulder_l}
    scale=<0 0 0>
    translation {z=-10}
   }
   n25 = "Transform (Modify) Bone - Bone: scale_shoulder_r"(ComponentPose=@n24, Alpha=1, Translation="-0.590,-0.000,-10.804", Rotation="0,0,0", Scale="0,0,0") {
    scaleMode=BMM_Additive
    scaleSpace=BCS_BoneSpace
    translationMode=BMM_Additive
    translationSpace=BCS_BoneSpace
    alphaBoolBlend {blendOption=Linear}
    boneToModify {boneName=scale_shoulder_r}
    scale=<0 0 0>
    translation {x=-0.59, z=-10.804}
   }
   n26 = "Transform (Modify) Bone - Bone: dyn_blouse_base_side_r"!(ComponentPose=@n25, bAlphaBoolEnabled:bool=True, Translation="0,0,-1.5", Rotation="0,0,0", Scale="1,1,1") -> (Pose) {
    alphaInputType=Bool
    translationMode=BMM_Additive
    translationSpace=BCS_BoneSpace
    alphaBoolBlend {blendOption=Linear}
    boneToModify {boneName=dyn_blouse_base_side_r}
    translation {z=-1.5}
   }
   n27 = "Transform (Modify) Bone - Bone: dyn_blouse_base_bk_r"!(ComponentPose=@n26, bAlphaBoolEnabled:bool=True, Translation="0,0,-2", Rotation="0,0,0", Scale="1,1,1") -> (Pose) {
    alphaInputType=Bool
    translationMode=BMM_Additive
    translationSpace=BCS_BoneSpace
    alphaBoolBlend {blendOption=Linear}
    boneToModify {boneName=dyn_blouse_base_bk_r}
    translation {z=-2}
   }
   n28 = Control Rig!(Source=Component To Local(ComponentPose=@n27), Alpha, Is rigid body:bool=Get is rigid body() -> (is rigid body:bool)) -> (Pose) {
    alphaBoolBlend {blendOption=Linear}
    controlRigAssetReference {blueprintRigClass=&/F_MED_ConfectionPop_CtrlRig.F_MED_ConfectionPop_CtrlRig_C}
   }
   Save cached pose 'Pre sim'(Pose=@n28) {cacheName=Pre sim}
  ]}
 EventGraph {nodes [
   n1 = Event BlueprintInitializeAnimation()
   n2 = Sequence!(execute=[@n1.then, @n22.then]) -> (then_0, then_1, then_2:Exec)
   n3 = GetPartSkeletalMeshComponent()
   n4 = Set Skeletal Mesh(execute=@n2.then_1, Skeletal Mesh:Skeletal Mesh Component Object Reference=@n3) -> (Output_Get:Skeletal Mesh Component Object Reference)
   > n5 = Get Dynamic Material Instance(SkeletalMeshComponent=@n4)
   > Set Material(Material:Material Instance Dynamic Object Reference=@n5) -> (Output_Get:Material Instance Dynamic Object Reference)
   > n6 = GetOrCreateDynamicOverlayMaterial(MeshComponent=@n3, MaterialName=None)
   > Set Overlay Material(Overlay Material:Material Instance Dynamic Object Reference=@n6) -> (Output_Get:Material Instance Dynamic Object Reference)
   n7 = CheckForBackbling()
   Set Is Cape(execute=@n2.then_0, Is Cape:bool=AND Boolean(A=@n7, B=@n7.bIsBackblingLong)) -> (Output_Get:bool)
   > Set Is backpack(Is backpack:bool=AND Boolean(A=@n7, B=NOT Boolean(A=@n7.bIsBackblingLong))) -> (Output_Get:bool)
   > Set Is NoBackpack(Is NoBackpack:bool=AND Boolean!(A=NOT Boolean(A=@n7), B=NOT Boolean(A=@n7.bIsBackblingACape), C:bool=NOT Boolean(A=@n7.bIsBackblingLong)) -> (ReturnValue)) -> (Output_Get:bool)
   Event BlueprintPostEvaluateAnimation() {is=title K2Node_Event}
   n8 = GetPartSkeletalMeshComponent()
   > Is Valid(InputObject=@n8)
   n9 = Get Rotation X Vector!(InRot:Rotator="-69,-11,11") -> (ReturnValue:Vector) {is=title K2Node_CallFunction}
   n10 = Get Rotation X Vector(InRot=GetBoneTransform!(
     self:Skinned Mesh Component Object Reference=@n8
     InBoneName:Name=upperarm_L
     TransformSpace:ERelativeTransformSpace Enum=RTS_ParentBoneSpace
    ) -> (ReturnValue_Location:Vector, ReturnValue_Rotation:Rotator, ReturnValue_Scale:Vector).ReturnValue_Rotation)
   n11 = Power!(Base:double="Clamp (Float)"(Value=Dot Product!(A:Vector=@n10, B:Vector=@n9) -> (ReturnValue:double), Min=0, Max=1), Exp:double=8) -> (ReturnValue:double) {is=title K2Node_CallFunction}
   > SetScalarParameterValueOnMaterials!(execute, self:Mesh Component Object Reference=@n8, ParameterName:Name=Arm_Blend_L, ParameterValue:float=@n11) -> (then)
   n12 = Get Rotation X Vector(InRot="-64,-9,14")
   n13 = Get Rotation X Vector(InRot=GetBoneTransform(self=@n8, InBoneName=upperarm_r, TransformSpace=RTS_ParentBoneSpace).ReturnValue_Rotation)
   n14 = Power(Base="Clamp (Float)"(Value=Dot Product(A=@n13, B=@n12), Min=0, Max=1), Exp=8)
   > SetScalarParameterValueOnMaterials(self=@n8, ParameterName=Arm_Blend_R, ParameterValue=@n14)
   > SetScalarParameterValueOnOverlayMaterial!(execute, MeshComponent:Mesh Component Object Reference=@n8, ParameterName:Name=Arm_Blend_L, ParameterValue:float=@n11) -> (then)
   > SetScalarParameterValueOnOverlayMaterial(MeshComponent=@n8, ParameterName=Arm_Blend_R, ParameterValue=@n14)
   n15 = Event BlueprintUpdateAnimation()
   > n16 = Sequence!(execute) -> (then_0, then_1, then_2, then_3:Exec)
   > n17 = Gravity Override State Function DEPRECATED(
    CharacterPartAnimInstance=Self-Reference()
    GameTimeInSeconds=GetGameTimeInSeconds()
    GravityOverrideParams=Get Gravity Override Params()
    DebugJointName=Pelvis
   )
   > Set Gravity Override(Gravity Override=@n17)
   Set is rigid body(execute=@n16.then_1, is rigid body:bool=NOT Boolean(A=Get bRigidBodyEnabled())) -> (Output_Get:bool)
   n18 = Get Skeletal Mesh() -> (Skeletal Mesh:Skeletal Mesh Component Object Reference)
   n19 = Get Overlay Material() -> (Overlay Material:Material Instance Dynamic Object Reference)
   n20 = Get Material() -> (Material:Material Instance Dynamic Object Reference)
   > Set WPO Globals Function(
    Material=@n20
    Overlay Material=@n19
    Skeletal Mesh=@n18
    "Local Noise Transform  Location"="Get Local Noise Transform  Location"() -> ("Local Noise Transform  Location":Vector)
    Local Noise Offset=Get Local Noise Offset()
   )
   > Set WPO Regional Function!(
    execute
    Material:Material Instance Dynamic Object Reference=@n20
    Overlay Material:Material Instance Dynamic Object Reference=@n19
    Skeletal Mesh:Skeletal Mesh Component Object Reference=@n18
    Velocity Transform:Name=lowerarm_r
    Velocity Interp Speed:double=Get Velocity Interp Speed()
    Collision Transform:Name=lowerarm_r
    Collision Transform Axis:EAxis Enum
    Invert Collision Transform Axis:bool=false
    Delta Time X:double=@n15.DeltaTimeX
    Previous Velocity Transform Location:"Vector (by ref)"=Get Previous Velocity Transform Location() -> (Previous Velocity Transform Location:Vector)
    Previous Velocity Transform Velocity:"Vector (by ref)"=Get Previous Velocity Transform Velocity() -> (Previous Velocity Transform Velocity:Vector)
    Region:int=1
   ) -> (then)
   > Set WPO Regional Function(
    Material=@n20
    Overlay Material=@n19
    Skeletal Mesh=@n18
    Velocity Transform=lowerarm_l
    Velocity Interp Speed=Get Velocity Interp Speed()
    Collision Transform=lowerarm_l
    Invert Collision Transform Axis=true
    Delta Time X=@n15.DeltaTimeX
    Previous Velocity Transform Location=Get Previous Velocity Transform Location_0() -> (Previous Velocity Transform Location_0:Vector)
    Previous Velocity Transform Velocity=Get Previous Velocity Transform Velocity_0() -> (Previous Velocity Transform Velocity_0:Vector)
    Region=2
   )
   > Set WPO Regional Function(
    Material=@n20
    Overlay Material=@n19
    Skeletal Mesh=@n18
    Velocity Transform=pelvis
    Velocity Interp Speed=Get Velocity Interp Speed()
    Collision Transform=pelvis
    Invert Collision Transform Axis=false
    Delta Time X=@n15.DeltaTimeX
    Previous Velocity Transform Location=Get Previous Velocity Transform Location_1() -> (Previous Velocity Transform Location_1:Vector)
    Previous Velocity Transform Velocity=Get Previous Velocity Transform Velocity_1() -> (Previous Velocity Transform Velocity_1:Vector)
    Region=3
   )
   > SetScalarParameterValue(self=[@n20, @n19], ParameterName=ClothSim_WindStrength, Value=Get WPO Wind() -> (WPO Wind:double))
   > SetScalarParameterValue(self=[@n20, @n19], ParameterName=ClothSim_TurbulenceNoiseTileDivisor, Value=Get Turbulence Noise() -> (Turbulence Noise:double))
   n21 = Branch(execute=@n16.then_2, Condition=Get bIsSkyDiving())
   > Set Velocity Interp Speed(Velocity Interp Speed:double=40) -> (Output_Get:double)
   > Set WPO Wind(WPO Wind:double=1) -> (Output_Get:double)
   > Set Turbulence Noise(Turbulence Noise:double=30) -> (Output_Get:double)
   Set Velocity Interp Speed(execute=@n21.else, Velocity Interp Speed=0)
   > Set WPO Wind(WPO Wind=0)
   > Set Turbulence Noise(Turbulence Noise=0)
   n22 = Event OnCharacterCustomizationLoaded()
   GetPlayerPawn!(PlayerIndex:int=0) -> (ReturnValue:Pawn Object Reference)
   n23 = Floor(A=float / float(A=1, B=@n15.DeltaTimeX)) {is=title K2Node_CallFunction}
   "To String (Float)"(InDouble=MapRangeClamped(Value="To Float (Integer)"(InInt=@n23), InRangeA=0, InRangeB=0, OutRangeB=0))
   GetSocketRotation!(self:Scene Component Object Reference=@n8, InSocketName:Name=upperarm_l) -> (ReturnValue:Rotator)
   PrintString!(
    execute
    InString=Append!(A="L: ", B="To String (Float)"(InDouble=@n11), C:String=" / R: ", D:String="To String (Float)"(InDouble=@n14)) -> (ReturnValue)
    bPrintToScreen
    bPrintToLog
    TextColor=(R=1,G=0.06,B=0,A=1)
    Duration=1
    Key
   ) -> (then)
   GetSocketRotation(self=@n8, InSocketName=upperarm_r)
   Sequence!(execute) -> (then_0, then_1)
  ]}
]
variables [
 Gravity Override {type=Vector}
 Gravity Override Params {
  type=GravityOverrideParamsStruct
  default {
   skydive_FloatingDown {gravityOverride=<0 2000 0>, windAmplitude=<0 100 200>, windFrequency=<20 35 35>, windOffset=<0 0 -300>}
   skydive_GoingUp {gravityOverride=<0 2000 0>, windAmplitude=<0 100 200>, windFrequency=<20 35 35>, windOffset=<0 0 1000>}
   skydive_Parachuting {windAmplitude=<50 -200 250>}
   tactical_Sprint {gravityOverride=<0 -980 -980>}
  }
 }
 Is backpack {type=boolean}
 Is Cape {type=boolean}
 Is NoBackpack {type=boolean}
 is rigid body {type=boolean}
 Local Noise Offset {type=Vector}
 "Local Noise Transform  Location" {type=Vector}
 Material {type=MaterialInstanceDynamic}
 Overlay Material {type=MaterialInstanceDynamic}
 Previous Velocity Transform Location {type=Vector}
 Previous Velocity Transform Location_0 {type=Vector}
 Previous Velocity Transform Location_1 {type=Vector}
 Previous Velocity Transform Velocity {type=Vector}
 Previous Velocity Transform Velocity_0 {type=Vector}
 Previous Velocity Transform Velocity_1 {type=Vector}
 Skeletal Mesh {type=SkeletalMeshComponent}
 Skydive Or Swimming {type=boolean}
 Turbulence Noise {type=number}
 UpperArmLRot {type=Rotator, default=<rot -25 -22.0833 25>}
 UpperArmRRot {type=Rotator, default=<rot -15.6774 -14.6375 19.3669>}
 Velocity Interp Speed {type=number}
 WPO Wind {type=number}
]
