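# Physics asset for the ConfectionPop female character's face accessory, defining capsule bodies and constraints for two pigtails, despite its CapePhysics name.
# Each pigtail base locks to the head and starts a six-segment chain with tight swing and little or no twist.
# Two short side branches fork from segment 04 on each side.
# Head, upper spine, clavicle and upper-arm capsules form the body; the head locks to spine_05, other body joints are free.
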PhysicsAsset /BRCosmetics/Characters/Player/Female/Medium/Bodies/F_MED_ConfectionPop/Meshes/Parts/F_MED_ConfectionPop_FaceAcc_CapePhysics
bodies {
 clavicle_l [{shapeType=Capsule, length=9.7797, radius=4.6502, shapeName=clavicle_l_capsule, center=<4.2235 0.2147 -0.255>, rotation=<rot -0.0006 -90 90>}]
 clavicle_r [{
   shapeType=Capsule
   length=9.8277
   radius=5.8558
   shapeName=clavicle_l_capsule
   center=<-4.2235 1.8254 0.4984>
   rotation=<rot -0.0006 -90 -90>
  }]
 dyn_pigtail_01_l [{shapeType=Capsule, length=0.9421, radius=4.5371, shapeName=dyn_pigtail_01_l_capsule, rotation=<rot 0 -90 -90>}]
 dyn_pigtail_01_r [{shapeType=Capsule, length=1.5087, radius=4.5863, shapeName=dyn_pigtail_01_l_capsule, rotation=<rot 0 -90 90>}]
 dyn_pigtail_02_l [{
   shapeType=Capsule
   length=0.3669
   radius=1.8544
   shapeName=dyn_pigtail_02_l_capsule
   center=<0.1303 0.3945 0.1644>
   rotation=<rot -23.8508 88.8911 -88.1024>
  }]
 dyn_pigtail_02_r [{
   shapeType=Capsule
   length=0
   radius=2.0417
   shapeName=dyn_pigtail_02_l_capsule
   center=<-0.1883 0.0116 -0.1715>
   rotation=<rot 0.9888 91.8007 86.0807>
  }]
 dyn_pigtail_03_l [{
   shapeType=Capsule
   length=12.6734
   radius=3.3073
   shapeName=dyn_pigtail_03_l_capsule
   center=<5.0184 1.3191 0.4614>
   rotation=<rot 77.1477 130.4578 -70.402>
  }]
 dyn_pigtail_03_r [{
   shapeType=Capsule
   length=12.3551
   radius=3.7911
   shapeName=dyn_pigtail_03_l_capsule
   center=<1.0914 0.375 -0.0261>
   rotation=<rot 3.9973 102.0434 90.1424>
  }]
 dyn_pigtail_04_l [{shapeType=Capsule, length=9, radius=3.5, shapeName=dyn_pigtail_04_l_capsule, rotation=<rot 0 120.1156 90>}]
 dyn_pigtail_04_r [{shapeType=Capsule, length=9, radius=3.5, shapeName=dyn_pigtail_04_r_capsule, center=<1.9074 -0.9165 0>, rotation=<rot 0 66.6287 90>}]
 dyn_pigtail_05_l [{shapeType=Capsule, length=9, radius=3.5, shapeName=dyn_pigtail_05_l_capsule, rotation=<rot 0 90 90>}]
 dyn_pigtail_05_r [{shapeType=Capsule, length=9, radius=3.5, shapeName=dyn_pigtail_05_r_capsule, rotation=<rot 0 90 90>}]
 dyn_pigtail_06_l [{shapeType=Capsule, length=9, radius=3.5, shapeName=dyn_pigtail_06_l_capsule, rotation=<rot 0 90 90>}]
 dyn_pigtail_06_r [{shapeType=Capsule, length=9, radius=3.5, shapeName=dyn_pigtail_06_r_capsule, rotation=<rot 0 90 90>}]
 dyn_pigtail_base_l [{shapeType=Capsule, length=0.505, radius=0.505, shapeName=dyn_pigtail_base_l_capsule, rotation=<rot 0 -90 0>}]
 dyn_pigtail_base_r [{shapeType=Capsule, length=0.505, radius=0.505, shapeName=dyn_pigtail_base_r_capsule, rotation=<rot 0 -90 0>}]
 dyn_pigtail_left_01_l [{
   shapeType=Capsule
   length=5.7491
   radius=1.8401
   shapeName=dyn_pigtail_left_01_l_capsule
   center=<2.1465 0.2359 -0.3345>
   rotation=<rot 7.3606 -89.5282 -88.1455>
  }]
 dyn_pigtail_left_02_l [{
   shapeType=Capsule
   length=6.2996
   radius=1.6944
   shapeName=dyn_pigtail_left_02_l_capsule
   center=<3.0401 -0.9628 0.8948>
   rotation=<rot -7.6908 71.3713 -70.3018>
  }]
 dyn_pigtail_reft_01_r [{shapeType=Capsule, length=7.2174, radius=2.683, shapeName=dyn_pigtail_reft_01_r_capsule, center=<3.9349 9 0>, rotation=<rot 0 90 90>}]
 dyn_pigtail_reft_02_r [{shapeType=Capsule, length=6.0947, radius=1.8037, shapeName=dyn_pigtail_reft_02_r_capsule, center=<4.7583 0 0>, rotation=<rot 0 90 90>}]
 dyn_pigtail_right_01_l [{shapeType=Capsule, length=4.1019, radius=1.3338, shapeName=dyn_pigtail_right_01_l_capsule, rotation=<rot 90 0 0>}]
 dyn_pigtail_right_01_r [{
   shapeType=Capsule
   length=6.7102
   radius=2.1517
   shapeName=dyn_pigtail_right_01_l_capsule
   center=<-1.1368 -2.1528 0>
   rotation=<rot 0 62.1643 90>
  }]
 dyn_pigtail_right_02_l [{shapeType=Capsule, length=3.0742, radius=1.4834, shapeName=dyn_pigtail_right_02_l_capsule, center=<1.5098 0 0>, rotation=<rot 90 0 0>}]
 dyn_pigtail_right_02_r [{
   shapeType=Capsule
   length=2.7782
   radius=1.0739
   shapeName=dyn_pigtail_right_02_l_capsule
   center=<2.6651 0.2119 0.1136>
   rotation=<rot -0.3532 -86.8391 86.3889>
  }]
 head [{shapeType=Capsule, length=6.087, radius=7.558, shapeName=head_capsule, center=<7.0092 1.0952 0>, rotation=<rot -90 -179.9605 179.9603>}]
 spine_03 [{shapeType=Capsule, length=11.0111, radius=12.1541, shapeName=spine_03_capsule, center=<-3.9683 -1.1008 0>, rotation=<rot 0 -90 0>}]
 spine_04 [{shapeType=Capsule, length=13.3559, radius=15.9974, shapeName=spine_04_capsule, center=<2.2987 -1.9499 0>, rotation=<rot 0 -90 0>}]
 spine_05 [{
   shapeType=Capsule
   length=17.4903
   radius=11.4093
   shapeName=spine_05_capsule
   center=<2.5253 -0.6945 0.3652>
   rotation=<rot 2.8769 -90 177.9076>
  }]
 upperarm_l [{
   shapeType=Capsule
   length=23.0715
   radius=7.1592
   shapeName=upperarm_l_capsule
   center=<14.0923 -0.5775 1.452>
   rotation=<rot -0.0002 -89.9998 85>
  }]
 upperarm_r [{
   shapeType=Capsule
   length=23.0715
   radius=6.2307
   shapeName=upperarm_l_capsule
   center=<-13.7789 0.2905 -1.4245>
   rotation=<rot -0.0002 -89.9998 -95>
  }]
}
constraints [
 {bone1Name=dyn_pigtail_03_l, bone2Name=dyn_pigtail_02_l, swing1LimitDegrees=13.32, swing2LimitDegrees=5.68, twistLimitDegrees=0}
 {bone1Name=dyn_pigtail_left_01_l, bone2Name=dyn_pigtail_04_l, swing1LimitDegrees=6, swing2LimitDegrees=4, twistLimitDegrees=2}
 {bone1Name=dyn_pigtail_left_02_l, bone2Name=dyn_pigtail_left_01_l, swing1LimitDegrees=10, swing2LimitDegrees=10, twistLimitDegrees=5}
 {bone1Name=dyn_pigtail_right_01_l, bone2Name=dyn_pigtail_04_l, swing1LimitDegrees=6, swing2LimitDegrees=4, twistLimitDegrees=2}
 {bone1Name=dyn_pigtail_right_02_l, bone2Name=dyn_pigtail_right_01_l, swing1LimitDegrees=5, swing2LimitDegrees=5, twistLimitDegrees=5}
 {
  bone1Name=head
  bone2Name=spine_05
  swing1LimitDegrees=45
  swing1Motion=Locked
  swing2LimitDegrees=45
  swing2Motion=Locked
  twistLimitDegrees=45
  twistMotion=Locked
 }
 {bone1Name=dyn_pigtail_right_01_r, bone2Name=dyn_pigtail_04_r, swing1LimitDegrees=6, swing2LimitDegrees=4, twistLimitDegrees=0}
 {bone1Name=dyn_pigtail_01_r, bone2Name=dyn_pigtail_base_r, swing1LimitDegrees=15, swing2LimitDegrees=2.88, twistLimitDegrees=0}
 {bone1Name=dyn_pigtail_02_l, bone2Name=dyn_pigtail_01_l, swing1LimitDegrees=5.76, swing2LimitDegrees=8.64, twistLimitDegrees=0}
 {bone1Name=dyn_pigtail_02_r, bone2Name=dyn_pigtail_01_r, swing1LimitDegrees=15, swing2LimitDegrees=10, twistLimitDegrees=0}
 {bone1Name=dyn_pigtail_03_r, bone2Name=dyn_pigtail_02_r, swing1LimitDegrees=39.6, swing2LimitDegrees=4.92, twistLimitDegrees=0}
 {bone1Name=dyn_pigtail_right_02_r, bone2Name=dyn_pigtail_right_01_r, swing1LimitDegrees=5, swing2LimitDegrees=5, twistLimitDegrees=0}
 {bone1Name=dyn_pigtail_reft_01_r, bone2Name=dyn_pigtail_04_r, swing1LimitDegrees=6, swing2LimitDegrees=4, twistLimitDegrees=2}
 {bone1Name=dyn_pigtail_reft_02_r, bone2Name=dyn_pigtail_reft_01_r, swing1LimitDegrees=5, swing2LimitDegrees=5, twistLimitDegrees=5}
 {
  bone1Name=upperarm_l
  bone2Name=clavicle_l
  swing1LimitDegrees=45
  swing1Motion=Free
  swing2LimitDegrees=45
  swing2Motion=Free
  twistLimitDegrees=45
  twistMotion=Free
 }
 {
  bone1Name=clavicle_l
  bone2Name=spine_05
  swing1LimitDegrees=45
  swing1Motion=Free
  swing2LimitDegrees=45
  swing2Motion=Free
  twistLimitDegrees=45
  twistMotion=Free
 }
 {
  bone1Name=clavicle_r
  bone2Name=spine_05
  swing1LimitDegrees=45
  swing1Motion=Free
  swing2LimitDegrees=45
  swing2Motion=Free
  twistLimitDegrees=45
  twistMotion=Free
 }
 {
  bone1Name=spine_05
  bone2Name=spine_04
  swing1LimitDegrees=45
  swing1Motion=Free
  swing2LimitDegrees=45
  swing2Motion=Free
  twistLimitDegrees=45
  twistMotion=Free
 }
 {
  bone1Name=spine_04
  bone2Name=spine_03
  swing1LimitDegrees=45
  swing1Motion=Free
  swing2LimitDegrees=45
  swing2Motion=Free
  twistLimitDegrees=45
  twistMotion=Free
 }
 {
  bone1Name=upperarm_r
  bone2Name=clavicle_r
  swing1LimitDegrees=45
  swing1Motion=Free
  swing2LimitDegrees=45
  swing2Motion=Free
  twistLimitDegrees=45
  twistMotion=Free
 }
 {
  bone1Name=dyn_pigtail_base_r
  bone2Name=head
  swing1LimitDegrees=45
  swing1Motion=Locked
  swing2LimitDegrees=45
  swing2Motion=Locked
  twistLimitDegrees=45
  twistMotion=Locked
 }
 {bone1Name=dyn_pigtail_01_l, bone2Name=dyn_pigtail_base_l, swing1LimitDegrees=22.12, swing2LimitDegrees=3.36, twistLimitDegrees=0}
 {
  bone1Name=dyn_pigtail_base_l
  bone2Name=head
  swing1LimitDegrees=45
  swing1Motion=Locked
  swing2LimitDegrees=45
  swing2Motion=Locked
  twistLimitDegrees=45
  twistMotion=Locked
 }
 {bone1Name=dyn_pigtail_04_l, bone2Name=dyn_pigtail_03_l, swing1LimitDegrees=10, swing2LimitDegrees=10, twistLimitDegrees=0}
 {bone1Name=dyn_pigtail_05_l, bone2Name=dyn_pigtail_04_l, swing1LimitDegrees=6, swing2LimitDegrees=4, twistLimitDegrees=2}
 {bone1Name=dyn_pigtail_06_l, bone2Name=dyn_pigtail_05_l, swing1LimitDegrees=6, swing2LimitDegrees=4, twistLimitDegrees=2}
 {bone1Name=dyn_pigtail_04_r, bone2Name=dyn_pigtail_03_r, swing1LimitDegrees=10, swing2LimitDegrees=4, twistLimitDegrees=0}
 {bone1Name=dyn_pigtail_05_r, bone2Name=dyn_pigtail_04_r, swing1LimitDegrees=6, swing2LimitDegrees=4, twistLimitDegrees=2}
 {bone1Name=dyn_pigtail_06_r, bone2Name=dyn_pigtail_05_r, swing1LimitDegrees=6, swing2LimitDegrees=4, twistLimitDegrees=2}
]
