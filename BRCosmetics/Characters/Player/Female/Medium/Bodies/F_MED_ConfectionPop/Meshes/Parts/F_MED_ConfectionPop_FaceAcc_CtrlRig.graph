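# Control Rig for the F_MED_ConfectionPop face accessory that applies corrective offsets to the pigtail bones as the head and upper spine move.
# The "RigidNode enabled" bool, false by default, selects between two corrective chains.
# False uses remapped or curve-evaluated head-to-spine_05 Euler angles; true combines a spherical pose reader with head and spine angle weights.
# Kalman filters smooth some weights, and several other pose-reader and transform nodes are left unwired from execution.
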ControlRigBlueprint /BRCosmetics/Characters/Player/Female/Medium/Bodies/F_MED_ConfectionPop/Meshes/Parts/F_MED_ConfectionPop_FaceAcc_CtrlRig
graphs [
 RigVMFunctionLibrary {links [], nodes []}
 RigVMModel {nodes [
   RigUnit_BeginExecution = &RigUnit_BeginExecution()
   RigVMFunction_ControlFlowBranch = &RigVMFunction_ControlFlowBranch(ExecuteContext=@RigUnit_BeginExecution.ExecutePin, Condition=RigVMVariableNode() {variable=RigidNode enabled}.Value)
   n1 = &RigUnit_GetRelativeTransformForItem(Child=(Type=Bone,Name="head"), Parent=(Type=Bone,Name="spine_05"), bChildInitial=False, bParentInitial=False)
   &RigUnit_ModifyTransforms(
    ItemToModify=((Item=(Type=Bone,Name="dyn_pigtail_base_l"),Transform=(Rotation=(X=-0.0373,Y=0.3588,Z=-0.2239,W=0.9054),Translation=(X=0,Y=0,Z=0),Scale3D=(X=1,Y=1,Z=1))))
    WeightMinimum=0
    Weight=&RigVMFunction_MathDoubleRemap(
     SourceMaximum=40
     SourceMinimum=20
     TargetMaximum=1
     TargetMinimum=0
     Value=&RigVMFunction_MathQuaternionToEuler(RotationOrder=ZYX, Value=@n1.RelativeTransform.Rotation).Result.X
    ).Result
    ExecutePin=@RigVMFunction_ControlFlowBranch.False
   )
   n2 = &RigVMFunction_MathQuaternionToEuler(
    RotationOrder=ZYX
    Value=&RigUnit_GetRelativeTransformForItem(Child=(Type=Bone,Name="head"), Parent=(Type=Bone,Name="spine_05"), bChildInitial=False, bParentInitial=False).RelativeTransform.Rotation
   )
   RigVMFunction_AnimEvalRichCurve = &RigVMFunction_AnimEvalRichCurve(
    Curve=(EditorCurveData=(Keys=((InterpMode=RCIM_Cubic,TangentMode=RCTM_User,ArriveTangent=2.8603,LeaveTangent=2.8603),(InterpMode=RCIM_Cubic,TangentMode=RCTM_User,Time=0.1822,Value=0.5752,ArriveTangent=1.1373,LeaveTangent=1.1373),(Time=1,Value=1))))
    SourceMaximum=-40
    SourceMinimum=-10
    TargetMinimum=0
    Value=@n2.Result.X
   )
   > &RigUnit_ModifyTransforms(
    ItemToModify=((Item=(Type=Bone,Name="dyn_pigtail_base_r"),Transform=(Rotation=(X=0.0387,Y=-0.3539,Z=-0.2142,W=0.9096),Translation=(X=0,Y=0,Z=0),Scale3D=(X=1,Y=1,Z=1))))
    WeightMinimum=0
    Weight=&RigVMFunction_KalmanFloat(BufferSize=7, Value=@RigVMFunction_AnimEvalRichCurve.Result) {template="KalmanFilter::Execute(in Value,in BufferSize,out Result)"}.Result
   )
   n3 = &RigVMFunction_MathDoubleRemap(SourceMaximum=-30, SourceMinimum=0, TargetMaximum=1, TargetMinimum=0, Value=@n2.Result.Z)
   > &RigUnit_ModifyTransforms(
    ItemToModify=((Item=(Type=Bone,Name="dyn_pigtail_base_r"),Transform=(Rotation=(X=0,Y=0,Z=-0.4867,W=0.8736),Translation=(X=0,Y=0,Z=0),Scale3D=(X=1,Y=1,Z=1))))
    WeightMinimum=0
    Weight=@n3.Result
   )
   > &RigUnit_ModifyTransforms(
    ItemToModify=((Item=(Type=Bone,Name="dyn_pigtail_01_l"),Transform=(Rotation=(X=-0.1309,Y=-0.1578,Z=-0.5346,W=0.8199),Translation=(X=0,Y=0,Z=0),Scale3D=(X=1,Y=1,Z=1))))
    WeightMinimum=0
    Weight=@n3.Result
   )
   > &RigUnit_ModifyTransforms(
    ItemToModify=((Item=(Type=Bone,Name="dyn_pigtail_base_r"),Transform=(Rotation=(X=-0.0864,Y=-0.1301,Z=0.0745,W=0.9849),Translation=(X=1.8519,Y=0.3283,Z=2.0614),Scale3D=(X=1,Y=1,Z=1))))
    WeightMinimum=0
    Weight=&RigVMFunction_MathDoubleRemap(SourceMaximum=-50, SourceMinimum=-30, TargetMaximum=1, TargetMinimum=0, Value=@n2.Result.Z).Result
   )
   RigVMFunction_Sequence = &RigVMFunction_Sequence(ExecuteContext=@RigVMFunction_ControlFlowBranch.True)
   n4 = &RigUnit_SphericalPoseReader(
    ActiveRegionScaleFactors=(PositiveWidth=3.5,NegativeWidth=1,PositiveHeight=2.5,NegativeHeight=3)
    ActiveRegionSize=0.05
    Debug=(bDrawDebug=true,bDraw2D=False,bDrawLocalAxes=False,DebugScale=25,DebugSegments=20,DebugThickness=0.25)
    DriverAxis=(X=1,Y=0.4,Z=0)
    DriverItem=(Type=Bone,Name="head")
    FalloffRegionScaleFactors=(PositiveWidth=5,NegativeWidth=1,PositiveHeight=2,NegativeHeight=3.5)
    FalloffSize=0.2
    FlipHeightScaling=False
    FlipWidthScaling=False
    OptionalParentItem=(Type=Bone,Name="spine_05")
    RotationOffset=(X=70,Y=0,Z=-130)
    ExecutePin=@RigVMFunction_Sequence.A
   )
   KalmanFilter = &RigVMFunction_KalmanFloat(BufferSize=15, Value=@n4.OutputParam) {template="KalmanFilter::Execute(in Value,in BufferSize,out Result)"}
   > &RigUnit_ModifyTransforms(
    ItemToModify=((Item=(Type=Bone,Name="dyn_pigtail_base_r"),Transform=(Rotation=(X=0,Y=0,Z=0,W=1),Translation=(X=0,Y=-3,Z=5),Scale3D=(X=1,Y=1,Z=1))))
    WeightMinimum=0
    Weight=@KalmanFilter.Result
    ItemToModify.0.Transform=/Script/RigVM.RigVMFunction_MathTransformFromSRT(Location=(X=0,Y=0,Z=0), Rotation=(X=0,Y=0,Z=-80), Scale=(X=1,Y=1,Z=1)) {is=struct RigVMUnitNode}.Transform
   )
   > &RigUnit_ModifyTransforms(
    ItemToModify=((Item=(Type=Bone,Name="dyn_pigtail_base_l"),Transform=(Rotation=(X=0,Y=0,Z=0,W=1),Translation=(X=0,Y=-3,Z=5),Scale3D=(X=1,Y=1,Z=1))))
    WeightMinimum=0
    Weight=@KalmanFilter.Result
    ItemToModify.0.Transform=&RigVMFunction_MathTransformFromSRT(Location=(X=0,Y=0,Z=0), Rotation=(X=0,Y=0,Z=-40), Scale=(X=1,Y=1,Z=1)).Transform
   )
   RigUnit_GetRelativeTransformForItem = &RigUnit_GetRelativeTransformForItem(Child=(Type=Bone,Name="spine_05"), Parent=(Type=Bone,Name="spine_04"), bChildInitial=False, bParentInitial=False)
   RigVMFunction_MathQuaternionToEuler = &RigVMFunction_MathQuaternionToEuler(RotationOrder=ZYX, Value=@RigUnit_GetRelativeTransformForItem.RelativeTransform.Rotation)
   Remap = &RigVMFunction_MathDoubleRemap(SourceMaximum=-9, SourceMinimum=0, TargetMaximum=1, TargetMinimum=0, Value=@RigVMFunction_MathQuaternionToEuler.Result.Z)
   Multiply = &RigVMFunction_MathDoubleMul(B=@Remap.Result, A=@n4.OutputParam)
   > &RigUnit_ModifyTransforms(
    ItemToModify=((Item=(Type=Bone,Name="dyn_pigtail_base_r"),Transform=(Rotation=(X=0,Y=0,Z=0,W=1),Translation=(X=0,Y=-3,Z=5),Scale3D=(X=1,Y=1,Z=1))))
    WeightMinimum=0
    Weight=&RigVMFunction_KalmanFloat(BufferSize=15, Value=@Multiply.Result) {template="KalmanFilter::Execute(in Value,in BufferSize,out Result)"}.Result
    ItemToModify.0.Transform=&RigVMFunction_MathTransformFromSRT(Location=(X=0,Y=1,Z=0), Rotation=(X=0,Y=-20,Z=20), Scale=(X=1,Y=1,Z=1)).Transform
   )
   n5 = &RigVMFunction_MathQuaternionToEuler(
    RotationOrder=ZYX
    Value=&RigUnit_GetRelativeTransformForItem(Child=(Type=Bone,Name="head"), Parent=(Type=Bone,Name="spine_05"), bChildInitial=False, bParentInitial=False).RelativeTransform.Rotation
   )
   n6 = &RigVMFunction_MathDoubleMul(
    A=&RigVMFunction_MathDoubleRemap(SourceMaximum=-45, SourceMinimum=-20, TargetMaximum=1, TargetMinimum=0, Value=@n5.Result.X).Result
    B=&RigVMFunction_MathDoubleRemap(SourceMaximum=-25, SourceMinimum=-10, TargetMaximum=1, TargetMinimum=0, Value=@n5.Result.Y).Result
   )
   > &RigUnit_ModifyTransforms(
    ItemToModify=((Item=(Type=Bone,Name="dyn_pigtail_base_r"),Transform=(Rotation=(X=0,Y=0,Z=0,W=1),Translation=(X=0,Y=-3,Z=5),Scale3D=(X=1,Y=1,Z=1))))
    WeightMinimum=0
    ItemToModify.0.Transform=&RigVMFunction_MathTransformFromSRT(Location=(X=0,Y=0,Z=0), Rotation=(X=0,Y=-20,Z=-40), Scale=(X=1,Y=1,Z=1)).Transform
    Weight=@n6.Result
   )
   > &RigUnit_ModifyTransforms(
    ItemToModify=((Item=(Type=Bone,Name="dyn_pigtail_base_l"),Transform=(Rotation=(X=0,Y=0,Z=0,W=1),Translation=(X=0,Y=-3,Z=5),Scale3D=(X=1,Y=1,Z=1))))
    WeightMinimum=0
    ItemToModify.0.Transform=&RigVMFunction_MathTransformFromSRT(Location=(X=0,Y=0,Z=0), Rotation=(X=-50,Y=0,Z=0), Scale=(X=1,Y=1,Z=1)).Transform
    Weight=@n6.Result
   )
   n7 = &RigUnit_SphericalPoseReader(
    ActiveRegionScaleFactors=(PositiveWidth=0.2,NegativeWidth=0.2,PositiveHeight=0.2,NegativeHeight=0.2)
    ActiveRegionSize=0.2
    Debug=(bDrawDebug=true,bDraw2D=False,bDrawLocalAxes=False,DebugScale=25,DebugSegments=20,DebugThickness=0.25)
    DriverAxis=(X=1,Y=0,Z=0)
    DriverItem=(Type=Bone,Name="head")
    ExecutePin=()
    FalloffRegionScaleFactors=(PositiveWidth=0.2,NegativeWidth=0.2,PositiveHeight=0.2,NegativeHeight=0.2)
    FalloffSize=0.4
    FlipHeightScaling=False
    FlipWidthScaling=False
    OptionalParentItem=(Type=Bone,Name="spine_05")
    RotationOffset=(X=60,Y=0,Z=220)
   )
   > &RigUnit_ModifyTransforms(
    ItemToModify=((Item=(Type=Bone,Name="dyn_pigtail_base_r"),Transform=(Rotation=(X=0,Y=0,Z=0,W=1),Translation=(X=0,Y=0,Z=0),Scale3D=(X=1,Y=1,Z=1))))
    WeightMinimum=0
    ItemToModify.0.Transform=&RigVMFunction_MathTransformFromSRT(Location=(X=-2,Y=-5,Z=-2), Rotation=(X=0,Y=0,Z=-60), Scale=(X=1,Y=1,Z=1)).Transform
    Weight=@n7.OutputParam
   )
   > n8 = &RigUnit_SphericalPoseReader(
    ActiveRegionScaleFactors=(PositiveWidth=0.2,NegativeWidth=0.2,PositiveHeight=0.2,NegativeHeight=0.2)
    ActiveRegionSize=0.2
    Debug=(bDrawDebug=true,bDraw2D=False,bDrawLocalAxes=False,DebugScale=25,DebugSegments=20,DebugThickness=0.25)
    DriverAxis=(X=1,Y=0,Z=0)
    DriverItem=(Type=Bone,Name="head")
    FalloffRegionScaleFactors=(PositiveWidth=0.2,NegativeWidth=0.2,PositiveHeight=0.2,NegativeHeight=0.2)
    FalloffSize=0.4
    FlipHeightScaling=False
    FlipWidthScaling=False
    OptionalParentItem=(Type=Bone,Name="spine_05")
    RotationOffset=(X=50,Y=0,Z=250)
   )
   > &RigUnit_ModifyTransforms(
    ItemToModify=((Item=(Type=Bone,Name="dyn_pigtail_base_r"),Transform=(Rotation=(X=0,Y=0,Z=0,W=1),Translation=(X=0,Y=0,Z=0),Scale3D=(X=1,Y=1,Z=1))))
    WeightMinimum=0
    ItemToModify.0.Transform=&RigVMFunction_MathTransformFromSRT(Location=(X=-2,Y=-5,Z=-2), Rotation=(X=0,Y=0,Z=-60), Scale=(X=1,Y=1,Z=1)).Transform
    Weight=@n8.OutputParam
   )
   &RigVMFunction_KalmanFloat(BufferSize=2, Value=@n3.Result) {template="KalmanFilter::Execute(in Value,in BufferSize,out Result)"}
   &RigUnit_ModifyTransforms(
    ExecutePin=()
    ItemToModify=((Item=(Type=Bone,Name="head"),Transform=(Rotation=(X=-0.0321,Y=-0.1325,Z=-0.0418,W=0.9898),Translation=(X=0,Y=0,Z=0),Scale3D=(X=1,Y=1,Z=1))))
    Weight=1
    WeightMinimum=0
   )
   RigUnit_GetTransform = &RigUnit_GetTransform(Item=(Type=Bone,Name="head"), bInitial=False)
   &RigUnit_ModifyTransforms(
    ExecutePin=()
    ItemToModify=((Item=(Type=Bone,Name="dyn_pigtail_base_r"),Transform=(Rotation=(X=0,Y=0,Z=0,W=1),Translation=(X=0,Y=-3,Z=5),Scale3D=(X=1,Y=1,Z=1))))
    WeightMinimum=0
    Weight=&RigVMFunction_MathDoubleRemap(
     SourceMaximum=130
     SourceMinimum=100
     TargetMaximum=1
     TargetMinimum=0
     Value=&RigVMFunction_MathQuaternionToEuler(RotationOrder=ZYX, Value=@RigUnit_GetTransform.Transform.Rotation).Result.Z
    ).Result
    ItemToModify.0.Transform=&RigVMFunction_MathTransformFromSRT(Location=(X=-2,Y=-5,Z=-2), Rotation=(X=0,Y=0,Z=-60), Scale=(X=1,Y=1,Z=1)).Transform
   )
   &RigUnit_GetRelativeTransformForItem(Child=(Type=Bone,Name="head"), Parent=(Type=Bone,Name="neck_02"), bChildInitial=False, bParentInitial=False)
   &RigVMFunction_MathTransformFromSRT(Location=(X=0,Y=-8,Z=0), Rotation=(X=0,Y=0,Z=-180), Scale=(X=1,Y=1,Z=1))
   &RigVMFunction_MathTransformFromSRT(Location=(X=0,Y=-8,Z=0), Rotation=(X=0,Y=0,Z=-180), Scale=(X=1,Y=1,Z=1))
  ]}
]
hierarchy [
 root {loc=<0 0 0>, rot=<0 0 0 1>}
 pelvis {parent=root, loc=<0 -0.5301 98.0571>, rot=<0.0127 0.707 -0.0127 -0.707>}
 spine_01 {parent=pelvis, loc=<6.1208 -0.9328 0>, rot=<0 0 0.034 -0.9994>}
 spine_02 {parent=spine_01, loc=<6.5996 -0.2095 0>, rot=<0 0 0.0238 -0.9997>}
 spine_03 {parent=spine_02, loc=<6.4914 -0.1237 0>, rot=<0 0 0.0149 -0.9999>}
 spine_04 {parent=spine_03, loc=<6.9598 0.0794 0>, rot=<0 0 -0.0132 -0.9999>}
 spine_05 {parent=spine_04, loc=<6.9987 0 0>, rot=<0 0 -0.0015 -1>}
 neck_01 {parent=spine_05, loc=<11.3138 0.1529 0>, rot=<0 0 -0.2003 -0.9797>}
 neck_02 {parent=neck_01, loc=<4.5627 0 0>, rot=<0 0 0.0504 -0.9987>}
 head {parent=neck_02, loc=<5.3914 0 0>, rot=<0 0 0.115 -0.9934>}
 dyn_bang_01_l {parent=head, loc=<6.2666 1.1107 -7.9198>, rot=<0.1738 0.6933 0.6934 0.0911>}
 dyn_bang_02_l {parent=dyn_bang_01_l, loc=<2.8581 0 0>, rot=<0 0 0 1>}
 dyn_bang_01_r {parent=head, loc=<6.2668 1.1107 7.9198>, rot=<-0.2701 -0.9612 0.0153 -0.0546>}
 dyn_bang_02_r {parent=dyn_bang_01_r, loc=<2.8579 0 0>, rot=<0.0097 0 0 -1>}
 dyn_pigtail_base_r {parent=head, loc=<3.2147 -5.1846 4.7404>, rot=<0.1448 -0.9598 0.0359 0.2377>}
 dyn_pigtail_01_r {parent=dyn_pigtail_base_r, loc=<0 0 0>, rot=<0 0 0 1>}
 dyn_pigtail_02_r {parent=dyn_pigtail_01_r, loc=<5.4629 0 0>, rot=<-0.1021 0.0597 0.2262 -0.9669>}
 dyn_pigtail_03_r {parent=dyn_pigtail_02_r, loc=<6.5235 1.1084 0>, rot=<0 0.0795 0 -0.9968>}
 dyn_pigtail_04_r {parent=dyn_pigtail_03_r, loc=<5.4628 1.0616 -0.4203>, rot=<0 0 -0.3559 -0.9345>}
 dyn_pigtail_reft_01_r {parent=dyn_pigtail_04_r, loc=<0.5389 -2.5312 -1.4307>, rot=<-0.0446 -0.1257 0.2622 -0.9558>}
 dyn_pigtail_reft_02_r {parent=dyn_pigtail_reft_01_r, loc=<7.9441 1.0415 -0.3313>, rot=<0 -0.129 0 -0.9916>}
 dyn_pigtail_right_01_r {parent=dyn_pigtail_04_r, loc=<2.6159 -1.7859 4.6383>, rot=<0 0 0 1>}
 dyn_pigtail_right_02_r {parent=dyn_pigtail_right_01_r, loc=<4.7826 0 -0.8448>, rot=<0 -0.1111 0 -0.9938>}
 dyn_pigtail_05_r {parent=dyn_pigtail_04_r, loc=<10.8885 -1.4454 -1.6639>, rot=<0 0 0 1>}
 dyn_pigtail_06_r {parent=dyn_pigtail_05_r, loc=<6.4644 -0.4506 1.0767>, rot=<0 0 0 1>}
 dyn_pigtail_base_l {parent=head, loc=<3.215 -5.1846 -4.7404>, rot=<0.179 -0.9686 -0.0313 -0.1696>}
 dyn_pigtail_01_l {parent=dyn_pigtail_base_l, loc=<0 0 0>, rot=<0 0 0 1>}
 dyn_pigtail_02_l {parent=dyn_pigtail_01_l, loc=<5.3608 0 0>, rot=<0.0369 -0.0125 0.0861 -0.9955>}
 dyn_pigtail_03_l {parent=dyn_pigtail_02_l, loc=<5.6349 -0.4277 0.0971>, rot=<0.0314 -0.0365 0.1416 -0.9888>}
 dyn_pigtail_04_l {parent=dyn_pigtail_03_l, loc=<6.817 2.0197 0.3229>, rot=<0.9208 0.3826 -0.0606 -0.0443>}
 dyn_pigtail_left_01_l {parent=dyn_pigtail_04_l, loc=<1.4239 3.5228 -4.089>, rot=<0.0228 -0.1025 -0.1906 -0.976>}
 dyn_pigtail_left_02_l {parent=dyn_pigtail_left_01_l, loc=<7.0517 -0.8747 0.4411>, rot=<0.0361 0.1639 0.2121 -0.9627>}
 dyn_pigtail_right_01_l {parent=dyn_pigtail_04_l, loc=<2.0865 1.659 3.0849>, rot=<0.0023 0.0856 0.0642 -0.9943>}
 dyn_pigtail_right_02_l {parent=dyn_pigtail_right_01_l, loc=<5.01 0 0>, rot=<0 0 0.0588 -0.9983>}
 dyn_pigtail_05_l {parent=dyn_pigtail_04_l, loc=<8.3208 1.1591 -0.9664>, rot=<0 0 0 1>}
 dyn_pigtail_06_l {parent=dyn_pigtail_05_l, loc=<6.8742 -0.4248 0.5769>, rot=<0 0 -0.0587 -0.9983>}
 clavicle_l {parent=spine_05, loc=<5.4831 -0.1053 -3.2946>, rot=<0.0465 -0.7289 -0.009 -0.6829>}
 upperarm_l {parent=clavicle_l, loc=<9.8658 -1.2136 -0.4105>, rot=<0.0164 -0.3245 -0.0269 -0.9454>}
 clavicle_r {parent=spine_05, loc=<5.4831 -0.1053 3.2946>, rot=<0.7289 0.0465 -0.6829 0.009>}
 upperarm_r {parent=clavicle_r, loc=<-9.8658 1.2136 0.4105>, rot=<0.0164 -0.3245 -0.0269 -0.9454>}
]
variables [RigidNode enabled {type=bool, default=False}]
