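# Physics asset for the ConfectionPop female character's face accessory mesh, set up for physics on two pigtails.
# Capsule bodies cover the pigtail bones plus head, upper spine, clavicles and upper arms.
# Each pigtail chains from a base on the head through six segments, with two short branches off the fourth, mostly under tight angular limits.
# Torso and arm joints rotate freely; head and right base are locked, while left base is free.
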PhysicsAsset /BRCosmetics/Characters/Player/Female/Medium/Bodies/F_MED_ConfectionPop/Meshes/Parts/F_MED_ConfectionPop_FaceAcc_Physics
bodies {
 clavicle_l [
  {shapeType=Capsule, length=9.7797, radius=7.4319, shapeName=clavicle_l_capsule, center=<4.2235 0.2147 0.0472>, rotation=<rot -0.0006 -90 90>}
  {shapeType=Capsule, length=23.1155, radius=6.8845, shapeName=clavicle_l_capsule, center=<15.2028 0 10.0796>, rotation=<rot -33.4037 0 0>}
 ]
 clavicle_r [
  {
   shapeType=Capsule
   length=9.8277
   radius=8.4691
   shapeName=clavicle_l_capsule
   center=<-4.2235 0.0321 -0.3686>
   rotation=<rot -0.0006 -90 -90>
  }
  {shapeType=Capsule, length=23.2881, radius=6.7119, shapeName=clavicle_r_capsule, center=<-16.0505 0 -11.5048>, rotation=<rot -28.5527 0 0>}
 ]
 dyn_pigtail_01_l [{shapeType=Capsule, length=0.9421, radius=4.5371, shapeName=dyn_pigtail_01_l_capsule, center=<0 0.5182 0.4254>, rotation=<rot 0 -90 -90>}]
 dyn_pigtail_01_r [{shapeType=Capsule, length=1.5087, radius=4.5863, shapeName=dyn_pigtail_01_l_capsule, center=<0 -0.9189 0.5135>, rotation=<rot 0 -90 90>}]
 dyn_pigtail_02_l [{
   shapeType=Capsule
   length=0.3669
   radius=1.8544
   shapeName=dyn_pigtail_02_l_capsule
   center=<0.1062 0.0792 0.8958>
   rotation=<rot -23.8508 88.8911 -88.1024>
  }]
 dyn_pigtail_02_r [{
   shapeType=Capsule
   length=0
   radius=2.0417
   shapeName=dyn_pigtail_02_l_capsule
   center=<0.76 -0.9331 0.4912>
   rotation=<rot 0.9888 91.8007 86.0807>
  }]
 dyn_pigtail_03_l [{
   shapeType=Capsule
   length=12.6734
   radius=3.3073
   shapeName=dyn_pigtail_03_l_capsule
   center=<6.8667 1.1183 0.7699>
   rotation=<rot 77.1477 130.4578 -70.402>
  }]
 dyn_pigtail_03_r [{
   shapeType=Capsule
   length=12.3551
   radius=3.7911
   shapeName=dyn_pigtail_03_l_capsule
   center=<4.7479 1.614 1.1807>
   rotation=<rot 3.8912 113.4048 100.6916>
  }]
 dyn_pigtail_04_l [{
   shapeType=Capsule
   length=6.5101
   radius=5.551
   shapeName=dyn_pigtail_04_l_capsule
   center=<2.0396 1.0471 -1.183>
   rotation=<rot -0.4156 -73.5473 -89.3033>
  }]
 dyn_pigtail_04_r [{
   shapeType=Capsule
   length=8.1052
   radius=5.2004
   shapeName=dyn_pigtail_04_r_capsule
   center=<2.523 0.2462 0.6846>
   rotation=<rot 0 -97.4803 -83.9519>
  }]
 dyn_pigtail_05_l [{
   shapeType=Capsule
   length=9.8385
   radius=5.2378
   shapeName=dyn_pigtail_05_l_capsule
   center=<6.5469 -0.0646 -0.0196>
   rotation=<rot 0 86.2971 -89.4005>
  }]
 dyn_pigtail_05_r [{
   shapeType=Capsule
   length=5.1212
   radius=2.5178
   shapeName=dyn_pigtail_05_r_capsule
   center=<5.9297 -0.0467 1.7549>
   rotation=<rot 0 91.9258 -69.1835>
  }]
 dyn_pigtail_06_l [{shapeType=Capsule, length=0.505, radius=0.505, shapeName=dyn_pigtail_06_l_capsule, rotation=<rot 0 -90 0>}]
 dyn_pigtail_06_r [{shapeType=Capsule, length=0.505, radius=0.505, shapeName=dyn_pigtail_06_r_capsule, rotation=<rot 0 -90 0>}]
 dyn_pigtail_base_l [{shapeType=Capsule, length=0.505, radius=0.505, shapeName=dyn_pigtail_base_l_capsule, rotation=<rot 0 -90 0>}]
 dyn_pigtail_base_r [{shapeType=Capsule, length=0.505, radius=0.505, shapeName=dyn_pigtail_base_r_capsule, rotation=<rot 0 -90 0>}]
 dyn_pigtail_left_01_l [{
   shapeType=Capsule
   length=5.7491
   radius=1.8401
   shapeName=dyn_pigtail_left_01_l_capsule
   center=<2.1465 0.2359 -0.3345>
   rotation=<rot 8.0897 -89.5044 -88.1423>
  }]
 dyn_pigtail_left_02_l [{
   shapeType=Capsule
   length=6.2996
   radius=1.6944
   shapeName=dyn_pigtail_left_02_l_capsule
   center=<2.9929 0.4259 0.5427>
   rotation=<rot -14.2191 91.9473 -74.2594>
  }]
 dyn_pigtail_reft_01_r [{
   shapeType=Capsule
   length=7.2174
   radius=2.683
   shapeName=dyn_pigtail_reft_01_r_capsule
   center=<3.8653 -0.5618 0.3491>
   rotation=<rot -1.1798 88.6344 -87.8202>
  }]
 dyn_pigtail_reft_02_r [{
   shapeType=Capsule
   length=6.0947
   radius=1.8037
   shapeName=dyn_pigtail_reft_02_r_capsule
   center=<4.7583 1.2582 -0.7613>
   rotation=<rot 3.7066 -71.8431 -73.3084>
  }]
 dyn_pigtail_right_01_l [{
   shapeType=Capsule
   length=4.1019
   radius=1.3338
   shapeName=dyn_pigtail_right_01_l_capsule
   center=<0.7979 -0.0895 0.1906>
   rotation=<rot -7.002 -75.5037 -85.7371>
  }]
 dyn_pigtail_right_01_r [{
   shapeType=Capsule
   length=4.7884
   radius=1.9952
   shapeName=dyn_pigtail_right_01_l_capsule
   center=<1.2314 -0.1243 -0.2252>
   rotation=<rot -30.1445 -97.7823 88.3708>
  }]
 dyn_pigtail_right_02_l [{
   shapeType=Capsule
   length=3.0742
   radius=1.4834
   shapeName=dyn_pigtail_right_02_l_capsule
   center=<0.5374 -0.0733 -0.0198>
   rotation=<rot 5.1463 -95.4051 -101.1094>
  }]
 dyn_pigtail_right_02_r [{
   shapeType=Capsule
   length=2.7782
   radius=1.0739
   shapeName=dyn_pigtail_right_02_l_capsule
   center=<2.6651 0.2119 0.1136>
   rotation=<rot -0.3532 -86.8391 86.3889>
  }]
 head [{shapeType=Capsule, length=6.087, radius=7.558, shapeName=head_capsule, center=<7.0092 1.0952 0>, rotation=<rot -90 -179.9605 179.9603>}]
 spine_03 [{shapeType=Capsule, length=11.0111, radius=12.1541, shapeName=spine_03_capsule, center=<-3.9683 -1.1008 0>, rotation=<rot 0 -90 0>}]
 spine_04 [{shapeType=Capsule, length=15.5438, radius=12.8177, shapeName=spine_04_capsule, center=<-1.6506 0.9267 0>, rotation=<rot 0 -90 0>}]
 spine_05 [{
   shapeType=Capsule
   length=8.7298
   radius=9.7883
   shapeName=spine_05_capsule
   center=<4.0936 0.6124 0.3569>
   rotation=<rot 2.8769 -90 177.9076>
  }]
 upperarm_l [{
   shapeType=Capsule
   length=23.0715
   radius=7.1592
   shapeName=upperarm_l_capsule
   center=<14.0923 -0.5775 1.452>
   rotation=<rot -0.0002 -89.9998 85>
  }]
 upperarm_r [{
   shapeType=Capsule
   length=23.0715
   radius=6.2307
   shapeName=upperarm_l_capsule
   center=<-13.7789 0.2905 -1.4245>
   rotation=<rot -0.0002 -89.9998 -95>
  }]
}
constraints [
 {bone1Name=dyn_pigtail_03_l, bone2Name=dyn_pigtail_02_l, swing1LimitDegrees=26.28, swing2LimitDegrees=10, twistLimitDegrees=14}
 {bone1Name=dyn_pigtail_left_01_l, bone2Name=dyn_pigtail_04_l, swing1LimitDegrees=3, swing2LimitDegrees=5, twistLimitDegrees=2}
 {bone1Name=dyn_pigtail_left_02_l, bone2Name=dyn_pigtail_left_01_l, swing1LimitDegrees=10, swing2LimitDegrees=10, twistLimitDegrees=5}
 {bone1Name=dyn_pigtail_right_01_l, bone2Name=dyn_pigtail_04_l, swing1LimitDegrees=3, swing2LimitDegrees=5, twistLimitDegrees=2}
 {bone1Name=dyn_pigtail_right_02_l, bone2Name=dyn_pigtail_right_01_l, swing1LimitDegrees=5, swing2LimitDegrees=5, twistLimitDegrees=5}
 {
  bone1Name=head
  bone2Name=spine_05
  swing1LimitDegrees=45
  swing1Motion=Locked
  swing2LimitDegrees=45
  swing2Motion=Locked
  twistLimitDegrees=45
  twistMotion=Locked
 }
 {bone1Name=dyn_pigtail_right_01_r, bone2Name=dyn_pigtail_04_r, swing1LimitDegrees=3, swing2LimitDegrees=5, twistLimitDegrees=2}
 {bone1Name=dyn_pigtail_01_r, bone2Name=dyn_pigtail_base_r, swing1LimitDegrees=15, swing2LimitDegrees=16.84, twistLimitDegrees=6}
 {bone1Name=dyn_pigtail_02_l, bone2Name=dyn_pigtail_01_l, swing1LimitDegrees=22.48, swing2LimitDegrees=5, twistLimitDegrees=11}
 {bone1Name=dyn_pigtail_02_r, bone2Name=dyn_pigtail_01_r, swing1LimitDegrees=44.6, swing2LimitDegrees=2, twistLimitDegrees=5}
 {bone1Name=dyn_pigtail_03_r, bone2Name=dyn_pigtail_02_r, swing1LimitDegrees=28.08, swing2LimitDegrees=10, twistLimitDegrees=10}
 {bone1Name=dyn_pigtail_right_02_r, bone2Name=dyn_pigtail_right_01_r, swing1LimitDegrees=5, swing2LimitDegrees=5, twistLimitDegrees=5}
 {bone1Name=dyn_pigtail_reft_01_r, bone2Name=dyn_pigtail_04_r, swing1LimitDegrees=3, swing2LimitDegrees=5, twistLimitDegrees=0}
 {bone1Name=dyn_pigtail_reft_02_r, bone2Name=dyn_pigtail_reft_01_r, swing1LimitDegrees=5, swing2LimitDegrees=5, twistLimitDegrees=0}
 {
  bone1Name=upperarm_l
  bone2Name=clavicle_l
  swing1LimitDegrees=45
  swing1Motion=Free
  swing2LimitDegrees=45
  swing2Motion=Free
  twistLimitDegrees=45
  twistMotion=Free
 }
 {
  bone1Name=clavicle_l
  bone2Name=spine_05
  swing1LimitDegrees=45
  swing1Motion=Free
  swing2LimitDegrees=45
  swing2Motion=Free
  twistLimitDegrees=45
  twistMotion=Free
 }
 {
  bone1Name=clavicle_r
  bone2Name=spine_05
  swing1LimitDegrees=45
  swing1Motion=Free
  swing2LimitDegrees=45
  swing2Motion=Free
  twistLimitDegrees=45
  twistMotion=Free
 }
 {
  bone1Name=spine_05
  bone2Name=spine_04
  swing1LimitDegrees=45
  swing1Motion=Free
  swing2LimitDegrees=45
  swing2Motion=Free
  twistLimitDegrees=45
  twistMotion=Free
 }
 {
  bone1Name=spine_04
  bone2Name=spine_03
  swing1LimitDegrees=45
  swing1Motion=Free
  swing2LimitDegrees=45
  swing2Motion=Free
  twistLimitDegrees=45
  twistMotion=Free
 }
 {
  bone1Name=upperarm_r
  bone2Name=clavicle_r
  swing1LimitDegrees=45
  swing1Motion=Free
  swing2LimitDegrees=45
  swing2Motion=Free
  twistLimitDegrees=45
  twistMotion=Free
 }
 {
  bone1Name=dyn_pigtail_base_r
  bone2Name=head
  swing1LimitDegrees=45
  swing1Motion=Locked
  swing2LimitDegrees=45
  swing2Motion=Locked
  twistLimitDegrees=45
  twistMotion=Locked
 }
 {bone1Name=dyn_pigtail_01_l, bone2Name=dyn_pigtail_base_l, swing1LimitDegrees=9.16, swing2LimitDegrees=5, twistLimitDegrees=0}
 {
  bone1Name=dyn_pigtail_base_l
  bone2Name=head
  swing1LimitDegrees=45
  swing1Motion=Free
  swing2LimitDegrees=45
  swing2Motion=Free
  twistLimitDegrees=45
  twistMotion=Free
 }
 {bone1Name=dyn_pigtail_04_l, bone2Name=dyn_pigtail_03_l, swing1LimitDegrees=3, swing2LimitDegrees=5, twistLimitDegrees=2}
 {bone1Name=dyn_pigtail_05_l, bone2Name=dyn_pigtail_04_l, swing1LimitDegrees=3, swing2LimitDegrees=5, twistLimitDegrees=2}
 {bone1Name=dyn_pigtail_06_l, bone2Name=dyn_pigtail_05_l, swing1LimitDegrees=3, swing2LimitDegrees=5, twistLimitDegrees=2}
 {bone1Name=dyn_pigtail_05_r, bone2Name=dyn_pigtail_04_r, swing1LimitDegrees=3, swing2LimitDegrees=5, twistLimitDegrees=2}
 {bone1Name=dyn_pigtail_04_r, bone2Name=dyn_pigtail_03_r, swing1LimitDegrees=3, swing2LimitDegrees=5, twistLimitDegrees=2}
 {bone1Name=dyn_pigtail_06_r, bone2Name=dyn_pigtail_05_r, swing1LimitDegrees=3, swing2LimitDegrees=5, twistLimitDegrees=2}
]
