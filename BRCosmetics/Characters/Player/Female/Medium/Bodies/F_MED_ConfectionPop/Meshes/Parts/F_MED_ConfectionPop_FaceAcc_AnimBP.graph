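# Animation Blueprint for the ConfectionPop female face accessory, posing and simulating its pigtails.
# It blends the parent pose with body mesh limb poses, then runs a Control Rig.
# During skydiving, additive bone offsets adjust the pigtail bases, and one of three RigidBody physics assets runs, depending on backbling type: none, backpack or cape.
# The event graph derives those flags and feeds movement-state gravity overrides into the simulation, with cape-specific parameters while moving.
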AnimBlueprint /BRCosmetics/Characters/Player/Female/Medium/Bodies/F_MED_ConfectionPop/Meshes/Parts/F_MED_ConfectionPop_FaceAcc_AnimBP
target_skeleton=&/F_MED_ConfectionPop_FaceAcc_Skeleton
graphs [
 AnimGraph {nodes [
   n1 = Layered blend per bone(
    BasePose=Copy Pose From Mesh() {bUseAttachedParent=true}
    BlendPoses_0=Copy Pose From Mesh(SourceMeshComponent=Get BodyMesh() -> (BodyMesh:Skeletal Mesh Component Object Reference)) {bUseAttachedParent=true}
   ) {layerSetup [{branchFilters [
       {boneName=upperarm_l}
       {boneName=upperarm_r}
       {boneName=lowerarm_l}
       {boneName=lowerarm_r}
       {boneName=thigh_l}
       {boneName=thigh_r}
       {boneName=calf_l}
       {boneName=calf_r}
      ]}]}
   n2 = Control Rig!(Source=@n1, Alpha, RigidNode enabled:bool=Get bRigidBodyEnabled()) -> (Pose) {
    alphaBoolBlend {blendOption=Linear}
    controlRigAssetReference {blueprintRigClass=&/F_MED_ConfectionPop_FaceAcc_CtrlRig.F_MED_ConfectionPop_FaceAcc_CtrlRig_C}
   }
   n3 = Get skydive Backpuck() -> (skydive Backpuck:bool)
   n4 = "Transform (Modify) Bone - Bone: dyn_pigtail_base_r"!(ComponentPose=Local To Component(LocalPose=@n2), bAlphaBoolEnabled:bool=@n3, Translation="0,0,0", Rotation="-20,40.057,0", Scale="1,1,1") -> (Pose) {
    alphaInputType=Bool
    rotationMode=BMM_Additive
    rotationSpace=BCS_BoneSpace
    alphaBoolBlend {blendInTime=0.5, blendOption=Linear, blendOutTime=0.5}
    boneToModify {boneName=dyn_pigtail_base_r}
    rotation {pitch=-20, yaw=40.057}
   }
   n5 = "Transform (Modify) Bone - Bone: dyn_pigtail_base_l"!(ComponentPose=@n4, bAlphaBoolEnabled:bool=@n3, Translation="0,0,0", Rotation="20,0,0", Scale="1,1,1") -> (Pose) {
    alphaInputType=Bool
    rotationMode=BMM_Additive
    rotationSpace=BCS_BoneSpace
    alphaBoolBlend {blendInTime=0.5, blendOption=Linear, blendOutTime=0.5}
    boneToModify {boneName=dyn_pigtail_base_l}
    rotation {pitch=20}
   }
   n6 = Get skydive no backpack() -> (skydive no backpack:bool)
   n7 = "Transform (Modify) Bone - Bone: dyn_pigtail_base_l"(ComponentPose=@n5, bAlphaBoolEnabled=@n6, Translation="0,0,0", Rotation="9.122,-35.917,-3.241", Scale="1,1,1") {
    alphaInputType=Bool
    rotationMode=BMM_Additive
    rotationSpace=BCS_BoneSpace
    alphaBoolBlend {blendInTime=0.5, blendOption=Linear, blendOutTime=0.5}
    boneToModify {boneName=dyn_pigtail_base_l}
    rotation=<rot 9.122 -35.917 -3.241>
   }
   n8 = "Transform (Modify) Bone - Bone: dyn_pigtail_base_r"(ComponentPose=@n7, bAlphaBoolEnabled=@n6, Translation="0,0,0", Rotation="-9.254,15.454,31.845", Scale="1,1,1") {
    alphaInputType=Bool
    rotationMode=BMM_Additive
    rotationSpace=BCS_BoneSpace
    alphaBoolBlend {blendInTime=0.5, blendOption=Linear, blendOutTime=0.5}
    boneToModify {boneName=dyn_pigtail_base_r}
    rotation=<rot -9.254 15.454 31.845>
   }
   n9 = Get skydive cape() -> (skydive cape:bool)
   n10 = "Transform (Modify) Bone - Bone: dyn_pigtail_base_l"(ComponentPose=@n8, bAlphaBoolEnabled=@n9, Translation="0,0,0", Rotation="0,-30,0", Scale="1,1,1") {
    alphaInputType=Bool
    rotationMode=BMM_Additive
    rotationSpace=BCS_BoneSpace
    alphaBoolBlend {blendInTime=0.5, blendOption=Linear, blendOutTime=0.5}
    boneToModify {boneName=dyn_pigtail_base_l}
    rotation {yaw=-30}
   }
   n11 = "Transform (Modify) Bone - Bone: dyn_pigtail_base_r"(ComponentPose=@n10, bAlphaBoolEnabled=@n9, Translation="0,0,0", Rotation="0,-35,0", Scale="1,1,1") {
    alphaInputType=Bool
    rotationMode=BMM_Additive
    rotationSpace=BCS_BoneSpace
    alphaBoolBlend {blendInTime=0.5, blendOption=Linear, blendOutTime=0.5}
    boneToModify {boneName=dyn_pigtail_base_r}
    rotation {yaw=-35}
   }
   n12 = Get Override World Gravity() -> (Override World Gravity:Vector)
   n13 = RigidBody!(
    ComponentPose=@n11
    bAlphaBoolEnabled:bool=Get Is NoBackpack() -> (Is NoBackpack:bool)
    OverrideWorldGravity:Vector=@n12
    bTransferBoneVelocities:bool=False
   ) -> (Pose) {
    alphaInputType=Bool
    bClampLinearTranslationLimitToRefPose=true
    bOverrideWorldGravity=true
    bTransferBoneVelocities=true
    overridePhysicsAsset=&/F_MED_ConfectionPop_FaceAcc_Physics
    simulationSpace=BaseBoneSpace
    alphaBoolBlend {blendOption=Linear}
    baseBoneRef {boneName=spine_05}
    componentAppliedLinearAccClamp=<3000 3000 3000>
    componentLinearAccScale=<0.4 0.4 0.4>
    componentLinearVelScale=<0.3 0.3 0.3>
    overrideWorldGravity {z=-980}
    simSpaceSettings {
     maxAngularAcceleration=100
     maxAngularVelocity=2
     maxLinearAcceleration=100
     maxLinearVelocity=180
     velocityScaleZ=0
     worldAlpha=0.65
     externalLinearDragV=<1 1 1>
    }
   }
   n14 = RigidBody(ComponentPose=@n13, bAlphaBoolEnabled=Get Is backpack() -> (Is backpack:bool), OverrideWorldGravity=@n12, bTransferBoneVelocities=False) {
    alphaInputType=Bool
    bClampLinearTranslationLimitToRefPose=true
    bOverrideWorldGravity=true
    bTransferBoneVelocities=true
    overridePhysicsAsset=&/F_MED_ConfectionPop_FaceAcc_BackpackPhysics
    simulationSpace=BaseBoneSpace
    alphaBoolBlend {blendOption=Linear}
    baseBoneRef {boneName=spine_05}
    componentAppliedLinearAccClamp=<3000 3000 3000>
    componentLinearAccScale=<0.4 0.4 0.4>
    componentLinearVelScale=<0.3 0.3 0.3>
    overrideWorldGravity {z=-980}
    simSpaceSettings {
     maxAngularAcceleration=100
     maxAngularVelocity=2
     maxLinearAcceleration=100
     maxLinearVelocity=180
     velocityScaleZ=0
     worldAlpha=0.65
     externalLinearDragV=<1 1 1>
    }
   }
   n15 = RigidBody(ComponentPose=@n14, bAlphaBoolEnabled=Get Is Cape() -> (Is Cape:bool), OverrideWorldGravity=@n12, bTransferBoneVelocities=False) {
    alphaInputType=Bool
    bClampLinearTranslationLimitToRefPose=true
    bOverrideWorldGravity=true
    bTransferBoneVelocities=true
    overridePhysicsAsset=&/F_MED_ConfectionPop_FaceAcc_CapePhysics
    simulationSpace=BaseBoneSpace
    alphaBoolBlend {blendOption=Linear}
    baseBoneRef {boneName=spine_05}
    componentAppliedLinearAccClamp=<3000 3000 3000>
    componentLinearAccScale=<0.4 0.4 0.4>
    componentLinearVelScale=<0.3 0.3 0.3>
    overrideWorldGravity {z=-980}
    simSpaceSettings {
     maxAngularAcceleration=100
     maxAngularVelocity=2
     maxLinearAcceleration=100
     maxLinearVelocity=180
     velocityScaleZ=0
     worldAlpha=0.65
     externalLinearDragV=<1 1 1>
    }
   }
   Output Pose(Result=Component To Local(ComponentPose=@n15)) {layerGroup=None}
  ]}
 EventGraph {nodes [
   Event BlueprintUpdateAnimation()
   > n1 = Gravity Override State Function DEPRECATED(CharacterPartAnimInstance=Self-Reference(), GameTimeInSeconds=GetGameTimeInSeconds(), GravityOverrideParams=Get Gravity Override Params())
   > Set Override World Gravity(Override World Gravity:Vector=@n1) -> (Output_Get:Vector)
   n2 = OR Boolean!(A=Get bIsSkyDiving(), B=Get bIsSkydiveDiveMode(), C:bool=Get bIsSkydiveFloating()) -> (ReturnValue)
   > Set skydive no backpack(skydive no backpack:bool=AND Boolean(A=@n2, B=Get Is NoBackpack())) -> (Output_Get:bool)
   > Set skydive Backpuck(skydive Backpuck:bool=AND Boolean(A=@n2, B=Get Is backpack())) -> (Output_Get:bool)
   > Set skydive cape(skydive cape:bool=AND Boolean(A=@n2, B=Get Is Cape())) -> (Output_Get:bool)
   > n3 = PlayerMovement(AnimInstance=Self-Reference(), XDotProductCoefficient=1, YDotProductCoefficient=1, ZDotProductCoefficient=1)
   > n4 = Branch(Condition=AND Boolean(A="float > float"(A=@n3.YMovement, B=0.5), B=Get Is Cape()))
   > Set Gravity Override Params(Gravity Override Params:Gravity Override Params Struct Structure=Get Gravity Override cape() -> (Gravity Override cape:Gravity Override Params Struct Structure)) -> (Output_Get:Gravity Override Params Struct Structure)
   Set Gravity Override Params(
    execute=@n4.else
    Gravity Override Params=Get Gravity Override standart() -> (Gravity Override standart:Gravity Override Params Struct Structure)
   )
   n5 = Event BlueprintInitializeAnimation()
   Set BodyMesh(execute=[@n5.then, @n7.then], BodyMesh=GetOwnerSkeletalMeshComponent())
   n6 = CheckForBackbling()
   > Set Is Cape(Is Cape:bool=AND Boolean(A=@n6, B=@n6.bIsBackblingLong)) -> (Output_Get:bool)
   > Set Is backpack(Is backpack:bool=AND Boolean(A=@n6, B=NOT Boolean(A=@n6.bIsBackblingLong))) -> (Output_Get:bool)
   > Set Is NoBackpack(Is NoBackpack:bool=AND Boolean!(A=NOT Boolean(A=@n6), B=NOT Boolean(A=@n6.bIsBackblingACape), C:bool=NOT Boolean(A=@n6.bIsBackblingLong)) -> (ReturnValue)) -> (Output_Get:bool)
   n7 = Event OnCharacterCustomizationLoaded()
  ]}
]
variables [
 BodyMesh {type=SkeletalMeshComponent}
 Gravity Override cape {
  type=GravityOverrideParamsStruct
  default {
   onGround_Moving {windOffset=<0 400 900>}
   skydive_FloatingDown {gravityOverride=<0 -1200 980>}
   skydive_GoingUp {gravityOverride=<0 -700 1500>}
  }
 }
 Gravity Override Params {
  type=GravityOverrideParamsStruct
  default {skydive_FloatingDown {gravityOverride=<0 -1200 980>}, skydive_GoingUp {gravityOverride=<0 -700 1500>}}
 }
 Gravity Override standart {
  type=GravityOverrideParamsStruct
  default {skydive_FloatingDown {gravityOverride=<0 -1200 980>}, skydive_GoingUp {gravityOverride=<0 -700 1500>}}
 }
 Is backpack {type=boolean}
 Is Cape {type=boolean}
 Is NoBackpack {type=boolean}
 Override World Gravity {type=Vector, default=<0 0 -980>}
 skydive Backpuck {type=boolean}
 skydive cape {type=boolean}
 skydive no backpack {type=boolean}
]
